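# Idle Niagara effect for the SaxoPop back bling, layering cold smoke and heart-named bling and sparkle emitters on a skeletal mesh.
# Cold_Smoke emits drifting, fading sub-UV sprites only while System.EnableIce is set.
# HeartBlings spawns short flickering flares at hand-placed offsets from a sampled mesh position, randomly mirrored left or right.
# Time of day scales size, color and alpha, while User.Variant picks flare tints.
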NiagaraSystem /BRCosmetics/Accessories/FORT_Backpacks/Backpack_F_MED_SaxoPop/FX/NS_Backpack_SaxoPop_Idle
dependencies {
 usedDataInterfaces [
  &NiagaraDataInterfaceArrayFloat3
  &NiagaraDataInterfaceCurve
  &NiagaraDataInterfaceArrayFloat
  &NiagaraDataInterfaceArrayColor
  /Script/Niagara.NiagaraDataInterfaceArrayBool
  &NiagaraDataInterfaceSkeletalMesh
  &NiagaraDataInterfaceColorCurve
 ]
 usedDynamicInputs [
  &Multiply_Vector2D_ByFloat
  &Lerp_Float
  &MakeVector2D
  /Niagara/DynamicInputs/Helpers/MaskBool
  &SetBoolByFloatComparison
  &VectorLength
 ]
 usedModules [
  &self:SystemSpawnScript.SystemScriptSource.SystemScriptGraph.NiagaraNodeAssignment_0.SetVariables_#1
  &SystemState
  &self:SystemSpawnScript.SystemScriptSource.SystemScriptGraph.NiagaraNodeAssignment_1.SetVariables_#2
  /BRCosmetics/Characters/Player/Female/Medium/Bodies/F_MED_PiperShelf_Pearl/FX/TimeOfDayGlobal
  &self:SystemSpawnScript.SystemScriptSource.SystemScriptGraph.NiagaraNodeAssignment_2.SetVariables_#3
 ]
}
emitters {
 Cold_Smoke {
  emitterUpdateScript [
   &EmitterState()
   &SpawnRate(
    SpawnRate=&MakeCustomFloatFromBool(Bool=$System.EnableIce, True Float=2, False Float=0)
    Use Spawn Probability=false
   )
  ]
  particleSpawnScript [
   &InitializeParticle(
    Color Maximum=<rgba 0.2908 0.4014 0.5781 0.15>
    Color Minimum=<rgba 0.0659 0.0911 0.2083 0.1>
    Color Mode=::NewEnumerator2
    Lifetime=4.5
    Position Mode=::NewEnumerator0
    Sprite Rotation Mode=::NewEnumerator0
    Sprite Size Mode=::NewEnumerator1
    Sprite UV Mode=::NewEnumerator3
    Uniform Sprite Size Max=60
    Uniform Sprite Size Min=55
    Write Material Random=true
   )
   &CameraOffset(Camera Offset Amount=5)
  ]
  particleUpdateScript [
   &ParticleState()
   &SkeletalMeshLocation(
    Bone Sampling Mode=::NewEnumerator4
    MaskSpawnGroup=false
    Orientation Sampling=::NewEnumerator1
    Sampled Position Offset=<0 5 0>
    Skeletal Mesh=$System.Skeletal Mesh
   )
   &ShapeLocation(MaskSpawnGroup=false, Override Seed=false, UseOwnerScale=true) {enabled=false}
   &ScaleSpriteSize(Scale Factor=&Vector2DFromFloat(Value=&FloatFromCurve(FloatCurve=DataInterface() {propertyValues {Curve {keys [
          {arriveTangent=13.6278, interpMode=RCIM_Cubic, leaveTangent=13.6278, leaveTangentWeight=0.3412, tangentMode=RCTM_User, time=0, value=0}
          {arriveTangent=0.9167, interpMode=RCIM_Cubic, leaveTangent=0.9167, time=0.112, value=1.0173}
          {
           arriveTangent=-0.5826
           arriveTangentWeight=0.2975
           interpMode=RCIM_Cubic
           leaveTangent=-0.5826
           leaveTangentWeight=0.1523
           time=0.6
           value=0.8
          }
          {arriveTangent=-0.75, interpMode=RCIM_Cubic, leaveTangent=-0.75, time=1, value=0.5}
         ]}}})))
   &ScaleColor(
    Scale Alpha=&FloatFromCurve(
     FloatCurve=DataInterface() {propertyValues {Curve {keys [
         {arriveTangent=1.138, interpMode=RCIM_Cubic, leaveTangent=1.138, tangentMode=RCTM_User, time=0, value=0}
         {
          arriveTangent=0.7235
          arriveTangentWeight=0.7806
          interpMode=RCIM_Cubic
          leaveTangent=-0.9759
          leaveTangentWeight=0.8245
          tangentMode=RCTM_Break
          time=0.5
          value=1
         }
         {tangentMode=RCTM_User, time=1, value=0}
        ]}}}
     Scale Curve=&FN_Select_Float_from_Time_Of_Day(Morning=0.9, Day=1, Evening=0.9, Night=0.65)
    )
    ScaleA=true
    ScaleRGB=true
   )
   &SubUVAnimation()
   &ScaleSpriteSize(
    Uniform Curve Scale=2
    Uniform Curve Sprite Scale=DataInterface() {propertyValues {Curve {keys [
        {arriveTangent=0.1945, interpMode=RCIM_Cubic, leaveTangent=0.1945, tangentMode=RCTM_User, time=0, value=0.35}
        {arriveTangent=1.9726, interpMode=RCIM_Cubic, leaveTangent=1.9726, tangentMode=RCTM_User, time=1, value=1}
       ]}}}
   ) {module=ScaleSpriteSize001}
   &SpriteRotationRate(Rotation Rate=&RandomRangeFloat(Minimum=-10, Maximum=10))
   &DynamicMaterialParameters(
    Index 0 Param 1=0.75
    Index 0 Param 2=&RandomRangeFloat(Minimum=0, Maximum=3)
    Index 0 Param 3=1
    Index 0 Param 4=1
    Module.Param0WriteEnabled=true
    Module.Param1WriteEnabled=true
    Module.Param2WriteEnabled=true
    Module.Param3WriteEnabled=true
   )
   /SetVariables(Particles.Position=&Add_Vector(
     A=Linked() {linkedVariable {name=Particles.Position, type=NiagaraPosition}}
     B=&Multiply_VectorByFloat(Vector=<0 0 -18>, Float=$Particles.NormalizedAge)
    )) {isSetParameters=true}
  ]
  rendererClasses [&NiagaraSpriteRendererProperties]
  renderers [{rendererClass=&NiagaraSpriteRendererProperties}]
 }
 HeartBlings {
  emitterUpdateScript [
   &EmitterState()
   &SpawnRate(Spawn Probability=0.5, SpawnRate=0.85, Use Spawn Probability=true)
  ]
  particleSpawnScript [
   &InitializeParticle(
    Color=<rgba 5 5 5 1>
    Lifetime Max=0.7
    Lifetime Min=0.5
    Lifetime Mode=::NewEnumerator1
    Sprite Size Mode=::NewEnumerator3
    Uniform Sprite Size=&RandomRangeFloat(Minimum=15, Maximum=25)
    UsePositionOffset=false
    Write Material Random=true
   )
   &CameraOffset(Camera Offset Amount=3)
   /SetVariables(Particles.Mirror:NiagaraBool=/Niagara/DynamicInputs/Arrays/SelectBoolFromArray(
     Array Sampling Mode:ENiagara_LifetimeMode=::NewEnumerator1
     Direct Array Index:NiagaraInt32=?
     Direct Array Mode:ENiagara_WrapClamp=::NewEnumerator0
     Evaluation Type:ENiagaraRandomnessEvaluation=::NewEnumerator0
     Recalculate Random Each Loop:NiagaraBool=false
     Randomness Mode:ENiagaraRandomnessMode=::NewEnumerator0
     Random Seed:NiagaraInt32=0
     Fixed Random Seed:NiagaraBool=false
     Bool Selection Array:NiagaraDataInterfaceArrayBool=DataInterface() {propertyValues {GpuSyncMode=SyncCpuToGpu, MaxElements=0, BoolData [false,true]}}
    )) {isSetParameters=true, module=SetVariables_#4}
  ]
  particleUpdateScript [
   &ParticleState()
   &SkeletalMeshLocation(
    Bone / Socket Index=$Particles.PartitionParticles.Partition
    Bone Sampling Mode=::NewEnumerator4
    MaskSpawnGroup=false
    Orientation Sampling=::NewEnumerator1
    Sampled Position Offset=&Multiply_Vector(
     A=&MergeHelper_VectorSwitch(
      0=&SelectVectorFromArray(
       Array Sampling Mode=::NewEnumerator0
       Direct Array Index=?
       Vector Selection Array=DataInterface() {propertyValues {FloatData [
          <20.166 -3.4049 2.7431>
          <15.2029 -4.8726 12.6575>
          <14.0276 5.2138 -4.7046>
          <9.9198 -8.4673 -7.2102>
          <-0.8108 2.4371 -12.142>
          <0.0561 -4.4411 -12.1456>
          <22.0664 0.2571 4.0837>
          <10.5096 3.3678 -14.504>
          <13.0242 6.5768 1.7442>
          <5.1675 -7.9007 10.5401>
          <17.9929 -0.7766 6.2635>
          <-1.081 -9.1323 -2.2206>
          <6.9059 10.271 -6.7415>
          <12.6784 6.0171 5.1392>
          <12.4295 -7.4273 -10.6136>
          <15.3472 -0.9828 15.3862>
          <9.5501 3.4056 14.15>
          <16.5812 -0.6923 11.0465>
          <12.3471 -8.1396 -4.0267>
          <20.6375 3.4021 -3.923>
          <5.9046 -8.9944 -5.722>
          <11.0892 -8.2767 6.4782>
          <15.4666 -5.2546 -10.4095>
          <6.0047 -9.3511 -2.2006>
          <4.6058 -9.4163 2.066>
          <5.1214 2.7309 -12.8633>
          <-2.0834 -8.5163 6.2584>
          <9.2061 8.6364 4.8759>
          <12.6014 3.735 -11.5559>
          <2.8959 -6.0907 12.0799>
          <11.8107 -7.7407 9.9776>
          <-3.0399 -5.9634 9.9914>
          <17.3962 -2.6568 -8.0285>
          <13.9282 -5.5703 15.8733>
          <9.0043 9.0165 -4.0112>
          <-1.2912 -2.0773 12.4829>
          <7.7502 8.0854 -9.3727>
          <9.5091 8.5527 -0.2365>
          <8.2489 -8.4682 8.6588>
          <13.2168 3.0736 12.2413>
          <7.3071 -7.39 12.9848>
          <15.8365 2.8995 -6.5126>
          <-0.7377 -9.1742 2.1846>
          <18.5771 -3.6453 -3.8646>
          <6.2344 2.0767 12.9752>
          <1.7805 -9.0814 4.9042>
          <10.686 6.1795 9.0058>
          <4.5709 6.9375 12.1749>
          <2.7475 11.3943 9.915>
          <0.6742 -7.9166 8.8666>
          <10.4434 -7.6775 14.0722>
          <13.8469 4.0116 8.2618>
          <6.3231 -8.3129 -9.5083>
          <7.9192 6.0405 -11.9587>
          <12.9118 -7.8954 4.3863>
          <6.5224 9.855 8.1373>
          <9.5808 -8.8172 3.9387>
          <6.4483 11.7274 3.2903>
          <16.1265 -2.5758 -12.6882>
          <4.0067 -8.5809 7.7639>
          <6.1144 11.174 -2.3396>
          <2.2175 2.3724 12.6648>
          <12.7369 -6.781 -14.9985>
          <6.097 -9.0504 5.2247>
          <18.0172 4.413 -3.0745>
          <19.584 5.3123 0.8068>
          <15.0692 -6.5278 -6.3871>
          <0.7324 7.0371 11.8603>
          <4.4826 -2.7919 12.694>
          <16.0372 1.3463 -9.5858>
          <17.1701 -4.6711 5.8536>
          <4.862 12.8181 6.3164>
          <10.7621 7.0632 -6.3225>
          <16.3128 1.6421 8.0522>
          <7.9491 7.1361 10.7248>
          <12.3369 5.051 -8.6696>
          <22.2105 -2.1806 -2.9003>
          <4.2806 11.6045 -9.1813>
          <15.5335 4.8113 -0.73>
          <18.5302 0.2153 -5.2703>
          <20.1765 4.3884 3.1151>
          <4.9887 -0.6479 -12.7159>
          <14.9551 1.2496 -13.1556>
          <1.5292 -0.5169 -12.6147>
          <16.4728 3.5337 4.2154>
          <-0.1357 -5.4111 11.7754>
          <15.0453 -2.9137 -16.7883>
          <5.6414 -6.4562 -12.8059>
          <12.2571 7.0078 -2.2064>
          <14.0141 1.2091 -16.2451>
          <9.1012 -7.7432 -13.516>
          <1.6259 6.4518 -12.4028>
          <4.5041 12.4332 -5.1369>
          <2.6035 -5.9371 -12.051>
          <15.5014 -5.6916 8.9413>
          <-2.9555 -6.0326 -10.4162>
          <1.6929 -8.1148 -8.6449>
          <1.7944 -9.1363 -4.5832>
          <-2.272 -8.3492 -6.7592>
          <-2.3023 -2.0133 -12.0007>
         ]}}
       Random Seed=0
      )
      1=&SelectVectorFromArray(
       Array Sampling Mode=::NewEnumerator0
       Direct Array Index=?
       Vector Selection Array=DataInterface() {propertyValues {FloatData [
          <21.2437 2.3982 -2.5169>
          <1.2331 -4.9844 -1.8283>
          <2.6364 -1.306 -4.5243>
          <19.417 3.0787 3.1377>
          <13.3441 3.9736 2.6439>
          <10.5818 -0.5759 -3.406>
          <23.5002 0.0015 4.0935>
          <0.516 1.7335 -4.7945>
          <23.8813 3.7209 -0.2736>
          <18.7551 -2.5869 -2.2875>
          <15.3703 -0.4984 -3.2425>
          <17.371 4.2994 0.5044>
          <13.8018 3.6837 -0.9711>
          <23.0445 -2.5399 -2.3821>
          <12.084 2.3624 4.262>
          <10.8814 -4.0263 -1.4>
          <18.8986 -3.0369 3.1956>
          <11.5586 2.2052 -2.3319>
          <26.3535 -3.5187 1.4901>
          <2.8462 3.5336 -2.8304>
          <8.323 1.6621 -2.8166>
          <5.9168 1.6824 -3.2433>
          <18.288 0.0327 4.3554>
          <7.719 -4.4579 3.6112>
          <13.3026 0.8084 -3.1677>
          <9.7731 -4.9064 1.5456>
          <15.8652 -2.9583 -2.2185>
          <3.9998 0.7166 -4.1469>
          <5.2724 -2.5477 -3.2446>
          <15.0071 -4.4206 -0.235>
          <18.3481 -4.1886 0.2139>
          <18.0405 0.6083 -3.1395>
          <5.7788 -5.0014 1.1284>
          <12.83 -2.6951 -2.7393>
          <15.7734 4.0441 2.2747>
          <5.6657 3.8478 -1.156>
          <8.0911 -1.7142 -3.3173>
          <18.9997 3.4393 -1.4175>
          <1.4166 5.2733 -0.7297>
          <7.7587 -4.2829 -1.1119>
          <11.8619 -0.2572 5.1643>
          <21.1846 -3.8662 -0.5129>
          <23.6719 -3.8623 0.7811>
          <16.1676 -4.1208 2.4975>
          <9.4164 0.075 5.2687>
          <6.6902 3.2394 3.7046>
          <11.6109 4.2996 0.8231>
          <25.3862 2.1913 3.5645>
          <9.7022 2.8237 3.9857>
          <5.9493 4.346 1.5029>
          <3.1166 4.96 1.8073>
          <6.8384 1.1014 5.0819>
          <8.8923 4.1005 2.0773>
          <16.025 2.6786 -2.1424>
          <9.8124 3.0971 -1.4875>
          <8.7384 3.9733 -0.1082>
          <23.9742 -2.4298 3.039>
          <26.0644 -1.3771 3.6514>
          <21.0314 4.0712 0.7466>
          <21.442 -2.8535 2.9573>
          <3.7309 -4.612 -1.4604>
          <14.6777 1.4162 4.4929>
          <4.394 -4.2737 4.1896>
          <0.6933 -2.4369 -4.5406>
          <20.636 -0.8151 -3.1493>
          <16.9743 2.208 3.9426>
          <20.7745 0.0034 4.152>
          <22.5287 2.6703 3.0396>
          <3.7159 2.8686 4.656>
          <26.0738 -3.3005 -1.7623>
          <24.9664 2.5969 -3.2539>
          <12.7224 -4.7423 1.26>
          <16.4317 -1.7046 4.379>
          <10.6875 -3.1803 4.7029>
          <13.5046 -3.6056 3.7651>
          <23.1173 0.529 -3.5755>
          <4.4151 -0.0884 5.5779>
          <14.0025 -1.1903 4.8381>
          <7.228 -2.055 5.4462>
          <25.4829 -0.885 -3.5629>
         ]}}
       Random Seed=0
      )
      2=&SelectVectorFromArray(
       Array Sampling Mode=::NewEnumerator0
       Direct Array Index=?
       Vector Selection Array=DataInterface() {propertyValues {FloatData [
          <22.8596 0.1142 3.4905>
          <4.9005 -5.0777 -1.1743>
          <0.5209 2.225 4.5506>
          <20.9645 1.0933 -3.911>
          <20.2642 -3.8819 0.5243>
          <18.0431 1.3865 2.8637>
          <9.152 0.4667 3.2475>
          <0.6276 -1.2402 5.035>
          <22.4552 -3.318 1.609>
          <11.9019 1.2529 2.9642>
          <17.4213 3.6717 1.1443>
          <25.6801 4.0948 1.7971>
          <15.3688 -3.9855 1.0232>
          <24.1887 -3.7586 -1.2822>
          <9.8047 2.7592 -4.0278>
          <8.5683 4.1982 -0.7765>
          <7.6336 2.8004 2.038>
          <19.9707 -1.3491 2.9349>
          <6.5808 -4.177 1.3478>
          <2.7833 5.1276 -0.3202>
          <14.2206 -2.1255 2.8826>
          <25.3925 1.1288 -4.032>
          <15.5916 -3.0948 -3.8241>
          <6.5324 -4.2138 -4.0935>
          <20.2313 4.1021 -0.1022>
          <11.0802 -4.8342 -1.2399>
          <14.8105 4.1598 -0.1208>
          <23.2687 3.0128 -2.6245>
          <1.1271 -4.1333 3.1095>
          <10.7285 3.235 1.2966>
          <9.8982 -3.3034 2.3132>
          <20.8853 2.3547 2.506>
          <12.3265 3.7025 -3.0313>
          <5.153 4.35 -2.0237>
          <18.2081 -3.0208 -3.3284>
          <19.5793 3.4323 -2.6824>
          <14.9872 0.7238 3.1848>
          <23.3619 3.6126 0.5616>
          <0.8505 4.6719 2.2211>
          <12.1226 4.2985 -0.7684>
          <12.6316 1.2083 -4.696>
          <11.6692 -1.4922 3.2593>
          <3.744 -4.3843 1.8481>
          <4.5672 3.5835 1.9786>
          <6.9152 3.0969 -3.857>
          <18.9212 -0.8647 -4.1779>
          <17.7951 -3.3473 1.6595>
          <3.1776 -1.4709 4.303>
          <15.573 -0.32 -4.7222>
          <14.854 2.5968 -3.9332>
          <17.6436 1.6959 -4.1275>
          <10.4014 -0.358 -5.266>
          <16.5366 4.0145 -2.2859>
          <14.6602 -4.6453 -1.3362>
          <16.9174 -1.3178 3.0362>
          <12.478 -4.0933 1.1209>
          <5.342 -2.3774 3.345>
          <7.8227 -1.8072 3.3053>
          <27.0209 -2.8345 -2.2512>
          <5.692 0.8309 3.6478>
          <1.8364 -5.5956 -0.1369>
          <17.9406 -4.3157 -1.1522>
          <3.1441 -5.0637 -3.4018>
          <2.9236 1.9843 4.0056>
          <24.5109 -2.3121 2.606>
          <21.2421 -3.1699 -2.757>
          <14.1158 2.8015 1.9589>
          <23.0776 -1.0562 -3.9034>
          <4.7083 1.4391 -5.0008>
          <8.9262 -4.5983 0.3785>
          <26.7128 -1.4306 3.1486>
          <10.1558 -3.8666 -4.2357>
          <12.907 -4.2249 -3.3375>
          <4.5895 -2.2437 -5.5012>
          <7.8835 -4.9279 -1.9441>
          <27.3285 -3.4583 1.0851>
          <7.5005 0.4694 -5.2493>
          <12.8834 -1.8146 -4.947>
          <7.9807 -2.1369 -5.3883>
          <26.1371 1.4648 3.8583>
         ]}}
       Random Seed=0
      )
      3=&SelectVectorFromArray(
       Array Sampling Mode=::NewEnumerator0
       Direct Array Index=?
       Vector Selection Array=DataInterface() {propertyValues {FloatData [
          <-0.7134 -2.5592 -2.2688>
          <1.9035 3.8382 -1.9808>
          <5.8898 1.8041 -3.748>
          <1.2049 0.8462 4.01>
          <7.4467 -1.7492 3.0881>
          <8.7381 -3.2894 0.776>
          <4.655 -1.1744 -3.6865>
          <9.2274 3.2904 -1.125>
          <3.2854 3.1133 1.7712>
          <5.397 0.938 3.4882>
          <9.1581 1.5886 2.2517>
          <2.6612 -3.1286 -2.0413>
          <8.873 -0.7019 -3.3754>
          <0.5914 0.7538 -3.9702>
          <6.1762 -3.6758 -1.0554>
          <6.7531 3.1739 1.2092>
          <4.0796 -2.4656 2.7296>
          <1.7874 -3.589 1.2459>
          <-0.2404 -2.577 2.6744>
          <5.1643 3.7213 -1.3073>
         ]}}
       Random Seed=0
      )
      4=&SelectVectorFromArray(
       Array Sampling Mode=::NewEnumerator0
       Direct Array Index=?
       Vector Selection Array=DataInterface() {propertyValues {FloatData [
          <1.6482 -3.4717 -1.4729>
          <1.6502 -2.4028 2.7046>
          <5.7008 -1.9999 3.1951>
          <0.0186 2.1891 -3.7534>
          <2.994 -1.2709 -3.6527>
          <-1.9634 -3.1101 0.9565>
          <-0.917 0.9741 3.7944>
          <2.1816 3.1653 2.8663>
          <9.2479 2.9621 -0.6501>
          <8.9447 0.2488 3.4929>
          <6.1604 3.4956 2.1671>
          <0.4548 4.0756 -0.1597>
          <8.7167 -3.0482 1.4666>
          <3.7438 0.3891 4.0631>
          <-1.2008 -1.8371 -3.0711>
          <5.4271 3.3822 -0.8665>
          <2.2932 3.1124 -1.9963>
          <7.6034 -0.6929 -3.3852>
          <4.8509 1.4991 -3.3289>
          <5.4971 -3.5381 -0.9096>
         ]}}
       Random Seed=0
      )
      5=&SelectVectorFromArray(
       Array Sampling Mode=::NewEnumerator0
       Direct Array Index=?
       Vector Selection Array=DataInterface() {propertyValues {FloatData [
          <40.4493 5.8816 -0.0327>
          <34.3185 -4.5179 4.8504>
          <36.155 4.789 4.6116>
          <20.0868 7.9111 -4.4596>
          <44.2205 -5.7562 1.0783>
          <31.5949 5.9981 4.6492>
          <37.2793 -0.0423 6.4061>
          <31.6285 -4.1674 -3.6821>
          <39.9677 -5.4604 0.7165>
          <32.8999 1.6497 -5.7017>
          <40.4069 -3.5991 -3.0132>
          <24.3978 -6.2564 -1.5867>
          <44.4667 -2.7765 -3.831>
          <39.2444 -4.4332 4.3703>
          <37.2192 4.1804 -3.6351>
          <35.5586 -5.3674 -1.0013>
          <25.5926 -3.2136 -5.5353>
          <34.1927 7.0895 -0.6746>
          <40.5763 3.6719 4.4766>
          <30.1723 -5.9297 2.9995>
          <45.1531 3.5242 3.1389>
          <43.1561 1.8823 -3.6835>
          <29.3736 6.6756 -3.8515>
          <44.5472 -3.2385 5.1313>
          <36.983 -1.5234 -4.6845>
          <28.0804 8.4836 1.061>
          <24.4535 5.3876 -6.3382>
          <42.4004 -0.0455 5.9695>
          <32.3515 -0.0865 6.9152>
          <24.3005 8.6201 -2.0147>
         ]}}
       Random Seed=0
      )
      6=&SelectVectorFromArray(
       Array Sampling Mode=::NewEnumerator0
       Direct Array Index=?
       Vector Selection Array=DataInterface() {propertyValues {FloatData [
          <34.8672 -5.6596 -0.0022>
          <40.8105 4.0587 2.8644>
          <43.1133 -5.5387 -0.2162>
          <16.2548 -1.4811 8.0577>
          <37.7509 1.7508 -6.0069>
          <24.2228 -6.4627 0.9203>
          <37.5885 -3.8701 -5.1987>
          <36.3818 0.1149 5.0101>
          <29.8183 7.5322 -3.2438>
          <26.0428 5.287 6.0001>
          <34.438 7.2203 -0.8283>
          <18.8275 7.4412 5.3537>
          <29.7285 7.4232 2.415>
          <38.4979 5.6654 -2.8489>
          <30.7106 -5.5245 -3.8212>
          <31.011 2.0103 6.0826>
          <22.1022 2.2622 7.7327>
          <30.5171 -3.9822 4.0798>
          <43.5147 -0.1365 4.1533>
          <32.4059 4.309 -5.7042>
          <38.9638 -3.9172 2.7995>
          <24.2956 8.7006 1.6907>
          <24.7609 -2.3533 6.2403>
          <43.9706 4.4055 -2.7724>
          <34.9783 5.05 3.5931>
          <19.4605 -5.1575 4.9347>
          <43.3996 -2.516 -5.6678>
          <43.7075 1.9038 -4.9383>
          <31.6381 -0.6866 -7.0112>
          <15.5294 4.04 8.2075>
         ]}}
       Random Seed=0
      )
      7=&SelectVectorFromArray(
       Array Sampling Mode=::NewEnumerator0
       Direct Array Index=?
       Vector Selection Array=DataInterface() {propertyValues {FloatData [
          <24.7577 3.6654 -2.7924>
          <14.8539 -6.0741 -4.0861>
          <19.2329 3.3667 -2.815>
          <31.9263 5.0897 1.9316>
          <30.2501 3.9985 -3.7984>
          <14.4247 -5.4042 4.3023>
          <30.97 1.3653 4.6409>
          <15.5342 -1.7908 -5.7286>
          <35.8945 4.6082 -5.3821>
          <36.9664 -5.8692 1.8682>
          <23.3073 -3.2859 4.7399>
          <45.8658 -6.9575 -0.0175>
          <43.3285 3.0518 7.6494>
          <39.4754 -1.4259 -7.2344>
          <15.2797 1.9878 -4.1532>
          <10.19 -6.2605 -3.0604>
          <11.1145 -0.7042 -5.152>
          <19.5044 -2.8516 -5.944>
          <15.9136 -7.8609 0.3617>
          <31.6861 -6.4519 0.7187>
          <19.0477 -6.1734 3.7739>
          <40.5984 7.9011 -4.7168>
          <14.8228 3.7858 0.4055>
          <39.313 3.6103 5.9829>
          <33.2791 5.8913 -1.6151>
          <10.6465 2.917 2.3028>
          <41.7519 9.9492 1.0271>
          <11.2621 -7.3291 1.0114>
          <21.4755 0.9463 -5.0798>
          <33.7162 -1.2265 -6.5198>
          <35.6371 0.462 5.6372>
          <25.2821 -1.6832 -6.0243>
          <34.4512 -5.5947 -3.2668>
          <19.431 -7.1667 -2.4682>
          <25.545 -7.0156 -1.5055>
          <40.3393 -6.1778 -1.6704>
          <28.9132 0.6052 -5.8854>
          <43.5259 -4.681 -5.7917>
          <26.8003 -6.2141 2.698>
          <12.2803 0.322 4.2757>
          <22.032 -7.3099 1.4044>
          <6.4943 1.029 -4.1846>
          <6.5204 -3.6204 -4.1703>
          <40.8578 -4.6096 4.0713>
          <27.5692 4.1867 1.367>
          <35.0591 4.3341 4.1968>
          <43.8609 -0.0236 -8.7914>
          <37.2309 7.598 -0.5834>
          <29.191 -5.5037 -4.0001>
          <24.1193 2.1725 3.2019>
          <39.6407 -0.8652 6.1287>
          <43.2644 5.2883 -8.3381>
          <38.7539 7.0586 3.4345>
          <10.7471 3.4819 -2.5542>
          <27.1593 -0.6881 4.8023>
          <16.1175 0.6979 3.9701>
          <42.5003 7.4682 5.7325>
          <23.1356 -5.3686 -4.6299>
          <45.2155 -5.1846 4.5699>
          <19.6053 3.2992 1.5412>
          <20.6179 0.3281 4.1964>
          <44.3021 -1.3976 7.1795>
          <30.1195 -3.8544 4.5413>
          <34.5205 -3.3181 4.8577>
          <10.8245 -3.4515 4.7976>
          <17.9841 -2.6966 5.019>
          <6.802 -2.3103 4.9176>
          <6.8644 4.3578 -0.2123>
          <6.7361 -6.1087 1.9701>
          <6.9146 1.9945 3.5026>
         ]}}
       Random Seed=0
      )
      8=&SelectVectorFromArray(
       Array Sampling Mode=::NewEnumerator0
       Direct Array Index=?
       Vector Selection Array=DataInterface() {propertyValues {FloatData [
          <33.4037 -2.4057 6.272>
          <16.5284 -0.9993 5.6868>
          <27.1045 -4.9113 -3.9573>
          <25.1369 1.1606 5.2414>
          <36.5974 2.1639 6.7667>
          <9.8363 3.9144 -0.7573>
          <10.2026 -6.4502 2.6899>
          <38.2353 -6.2559 0.337>
          <31.6592 2.4995 5.4566>
          <28.4351 4.5939 1.796>
          <27.385 1.0678 -4.2808>
          <44.3999 -0.0194 8.7906>
          <43.751 0.8674 -7.5519>
          <40.0749 8.9558 1.2998>
          <30.1171 -6.5691 -0.9289>
          <11.9175 0.5207 4.8481>
          <34.0518 -6.2229 1.7546>
          <33.0961 -3.594 -4.7084>
          <20.7392 0.2799 5.3233>
          <25.1354 3.6504 -1.6709>
          <16.5213 3.0779 3.0931>
          <40.9269 -0.4491 7.7303>
          <17.5487 -5.4848 4.7687>
          <41.5793 -4.4991 5.4829>
          <28.3535 -1.4803 6.1839>
          <23.8741 -7.1036 -1.605>
          <40.3269 -5.2666 -3.0165>
          <11.2599 3.3322 2.7718>
          <37.7837 -1.4899 -5.6637>
          <6.8664 2.9708 -2.6077>
          <42.1787 -6.2964 1.6103>
          <30.5181 -0.9609 -5.0536>
          <6.598 2.1532 3.6965>
          <12.2438 -3.5376 5.222>
          <18.6677 3.4017 -1.3697>
          <33.9857 6.0283 2.1188>
          <37.7167 -2.3312 6.6174>
          <38.112 6.3454 4.8379>
          <21.9479 4.0408 1.4175>
          <16.1579 -7.1043 -2.686>
          <13.9942 2.9024 -2.1099>
          <30.6627 -5.4966 3.8898>
          <35.4956 -5.4916 -2.6194>
          <31.3947 4.3082 -2.7897>
          <6.5486 -6.0957 1.6258>
          <45.3992 -4.9712 5.9135>
          <45.8802 -6.7156 0.3124>
          <37.3652 -5.2327 4.0513>
          <18.1881 0.4853 -4.0908>
          <6.9214 -5.3383 -3.2006>
          <39.82 4.6349 -5.8923>
          <44.9299 -4.3581 -4.9141>
          <45.5497 -6.1123 -2.4801>
          <21.3085 -4.3011 -4.6419>
          <14.1746 -7.5753 1.5116>
          <16.7706 -3.6503 -4.9504>
          <41.5614 -2.138 -5.9782>
          <6.3979 -2.1849 4.5608>
          <41.6929 4.4399 7.6924>
          <19.3835 -7.6854 0.4484>
          <11.4904 -5.69 -3.8408>
          <41.6185 8.9216 -3.5344>
          <35.0591 1.6817 -5.3353>
          <36.0824 5.9993 -3.0395>
          <10.3117 0.6628 -4.1318>
          <22.9248 0.0395 -4.3299>
          <6.9565 -0.8714 -4.9245>
          <22.3877 -5.8661 4.2119>
          <13.5499 -1.167 -4.8321>
          <26.3859 -5.9977 3.7086>
         ]}}
       Random Seed=0
      )
      Index=$Particles.PartitionParticles.Partition
      Num=::NewEnumerator7
     )
     B=&SelectVectorFromArray(
      Direct Array Index=$Particles.PartitionParticles.Partition
      Vector Selection Array=DataInterface() {propertyValues {FloatData [<1 1 1>, <1 -1 -1>, <-1 -1 -1>, <1 1 1>, <-1 -1 1>, <-1 1 1>, <1 -1 1>, <-1 1 1>, <1 -1 1>]}}
     )
    )
    Skeletal Mesh=$System.Skeletal Mesh
   ) {enabled=false, module=SkeletalMeshLocation002}
   &SampleSkeletalMesh(
    AdditionalYawPitchRoll=false
    Bone Sampling Mode:ENiagaraBoneSamplingMode=::NewEnumerator4
    Skeletal Mesh:NiagaraDataInterfaceSkeletalMesh=$System.Skeletal Mesh
   )
   &ShapeLocation(
    Handle Radius=1
    Large Radius=12
    MaskSpawnGroup=false
    Non Uniform Scale=&MakeCustomVectorFromBool(Bool=$Particles.Mirror, True Vector=<1 -1.5 1>, False Vector=<-1 -1.5 1>)
    Offset=&MakeCustomVectorFromBool(Bool=$Particles.Mirror, True Vector=<-0.9794 -7.9938 5.8326>, False Vector=<0.9794 -7.9938 5.8326>)
    Offset Mode=::NewEnumerator0
    Override Seed=false
    Rotation Angle=90
    Rotation Axis=<0 1 0>
    Rotation Mode=::NewEnumerator1
    Shape Origin=$Particles.SampleSkeletalMesh.SampledPosition
    Shape Primitive=::NewEnumerator3
    Surface Distribution=0.5
    U Distribution=0.8
    UseOwnerScale=true
    V Distribution=0.4
   ) {enabled=false}
   &ShapeLocation(
    MaskSpawnGroup=false
    Offset=&MakeCustomVectorFromBool(
     Bool=$Particles.Mirror
     True Vector=&SelectVectorFromArray(
      Array Sampling Mode=::NewEnumerator0
      Direct Array Index=?
      Vector Selection Array=DataInterface() {propertyValues {FloatData [
         <5.0396 -0.9147 4.6938>
         <10.3826 -9 -0.2261>
         <9.08 -9 -3.4551>
         <6.8852 -9 -6.8791>
         <4.3545 -9 -9.9046>
         <1.6421 -9 -12.1911>
         <10.3983 -8.5 2.8509>
         <10.3983 -7.6495 5.2744>
         <9.3081 -5.561 7.4836>
         <7.6758 -2.8551 7.045>
        ]}}
      Random Seed=0
     )
     False Vector=&SelectVectorFromArray(
      Array Sampling Mode=::NewEnumerator0
      Direct Array Index=?
      Vector Selection Array=DataInterface() {propertyValues {FloatData [
         <-5.0396 -0.9147 4.6938>
         <-10.3826 -9 -0.2261>
         <-9.08 -9 -3.4551>
         <-6.8852 -9 -6.8791>
         <-4.3545 -9 -9.9046>
         <-1.6421 -9 -12.1911>
         <-10.3983 -8.5 2.8509>
         <-10.3983 -7.6495 5.2744>
         <-9.3081 -5.561 7.4836>
         <-7.6758 -2.8551 7.045>
        ]}}
      Random Seed=0
     )
    )
    Offset Mode=::NewEnumerator0
    Override Seed=false
    Shape Origin=$Particles.SampleSkeletalMesh.SampledPosition
    Sphere Radius=1
    UseOwnerScale=true
   ) {module=ShapeLocation001}
   &ScaleSpriteSize(Uniform Curve Scale=&FNTimeOfDayScalar(Night Scale=1.25, Morning Scale=1, Evening Scale=1)) {module=ScaleSpriteSize001}
   &ScaleColor(
    Scale Alpha=&Multiply_Float(
     A=&FloatFromCurve(
      FloatCurve=DataInterface() {propertyValues {Curve {keys [
          {interpMode=RCIM_Cubic, tangentMode=RCTM_User, time=0, value=0}
          {arriveTangent=-2.4058, interpMode=RCIM_Cubic, leaveTangent=-2.4058, tangentMode=RCTM_User, time=0.26, value=0.8421}
          {arriveTangent=-1.1775, interpMode=RCIM_Cubic, leaveTangent=-1.1775, time=0.4613, value=0.0645}
          {arriveTangent=-0.1197, interpMode=RCIM_Cubic, leaveTangent=-0.1197, time=0.5898, value=0.4537}
          <at 1 0 RCIM_Cubic>
         ]}}}
      Scale Curve=&RandomRangeFloat(Minimum=0.1, Maximum=0.5)
     )
     B=&FloatFromCurve(
      FloatCurve=DataInterface() {propertyValues {Curve {keys [{tangentMode=RCTM_User, time=0, value=0}, {tangentMode=RCTM_User, time=1, value=1}]}}}
      Scale Curve=&FloatFromCurve(
       FloatCurve=DataInterface() {propertyValues {Curve {keys [
           {tangentMode=RCTM_User, time=0, value=0}
           {arriveTangentWeight=0.1667, interpMode=RCIM_Cubic, leaveTangentWeight=0.1667, tangentMode=RCTM_User, time=1, value=1}
          ]}}}
       CurveIndex=$Emitter.Age
      )
     )
    )
    Scale RGB=&VectorFromFloat(Value=&Multiply_Float(
      A=&RandomRangeFloat(Minimum=0.7, Maximum=0.9, Evaluation Type=::NewEnumerator1)
      B=&SelectFloatFromArray(Direct Array Index=$Engine.QualityLevel, Float Selection Array=DataInterface() {propertyValues {FloatData [1,1,1,1,1]}})
     ))
    ScaleA=true
    ScaleRGB=true
   ) {module=ScaleColor001}
   &ScaleColor(
    Scale Alpha=$System.TODAlpha
    Scale RGB=&MakeVectorFromLinearColorRGB(Color=$System.TODColor)
    ScaleA=true
    ScaleRGB=true
   )
   /SetVariables(
    Particles.CenterFlareColor:LinearColor=&MultiplyLinearColor(
     Linear Color=$Particles.Color
     Scale Factor=&MakeVector4_FromLinearColor(LinearColor=&SelectLinearColorFromArray(
       Array Sampling Mode=::NewEnumerator1
       Direct Array Index=&MakeIntFromFloat(Float=$User.Variant)
       Color Selection Array=DataInterface() {propertyValues {ColorData [
          <rgba 8.0099 9.8007 10 1>
          <rgba 8.0099 9.8007 10 1>
          <rgba 6.8097 6.5104 10 1>
          <rgba 10 9.3031 3.947 1>
          <rgba 10 10 10 1>
          <rgba 8.0099 9.8007 10 1>
         ]}}
       Random Seed=?
      ))
    )
    Particles.CenterFlareSize:Vector2f=&Multiply_Vector2(A=$Particles.SpriteSize, B=&Vector2DFromFloat(Value=0.5))
    Particles.CoronaFlareColor:LinearColor=&MultiplyLinearColor(
     Linear Color=$Particles.Color
     Scale Factor=&MakeVector4_FromLinearColor(LinearColor=<rgba 0.15 0.15 0.15 1>)
    )
    Particles.CoronaFlareSize:Vector2f=&Multiply_Vector2(A=$Particles.SpriteSize, B=&Vector2DFromFloat(Value=0.75))
    Particles.FlickerFlareColor:LinearColor=&MultiplyLinearColor(
     Linear Color=$Particles.Color
     Scale Factor=&MakeVector4_FromLinearColor(LinearColor=&SelectLinearColorFromArray(
       Array Sampling Mode=::NewEnumerator1
       Direct Array Index=&MakeIntFromFloat(Float=$User.Variant)
       Color Selection Array=DataInterface() {propertyValues {ColorData [
          <rgba 0.9441 0.8079 1 1>
          <rgba 0.9441 0.8079 1 1>
          <rgba 0.0352 0 0.75 1>
          <rgba 0.1 0.0688 0.01 1>
          <rgba 0.75 0.75 0.75 1>
          <rgba 0.6059 0.6561 0.75 1>
         ]}}
       Random Seed=?
      ))
    )
   ) {isSetParameters=true, module=SetVariables_#5}
  ]
  rendererClasses [&NiagaraSpriteRendererProperties, &NiagaraMeshRendererProperties]
  renderers [
   {rendererClass=&NiagaraSpriteRendererProperties}
   {rendererClass=&NiagaraSpriteRendererProperties, rendererIndex=1}
   {rendererClass=&NiagaraSpriteRendererProperties, rendererIndex=2}
   {rendererClass=&NiagaraMeshRendererProperties, rendererIndex=3}
  ]
 }
 HeartSparkles {
  emitterUpdateScript [
   &EmitterState()
   &SpawnRate(SpawnRate=1.2, Use Spawn Probability=true)
  ]
  particleSpawnScript [
   &InitializeParticle(
    Color=<rgba 5 5 5 1>
    Lifetime Max=0.6
    Lifetime Min=0.4
    Lifetime Mode=::NewEnumerator1
    Sprite Size Mode=::NewEnumerator3
    Uniform Sprite Size=&RandomRangeFloat(Minimum=15, Maximum=25)
    UsePositionOffset=false
    Write Material Random=true
   )
   &CameraOffset(Camera Offset Amount=3)
   /SetVariables(Particles.Mirror:NiagaraBool=&SelectBoolFromArray(
     Array Sampling Mode:ENiagara_LifetimeMode=::NewEnumerator1
     Direct Array Index:NiagaraInt32=?
     Direct Array Mode:ENiagara_WrapClamp=::NewEnumerator0
     Evaluation Type:ENiagaraRandomnessEvaluation=::NewEnumerator0
     Recalculate Random Each Loop:NiagaraBool=false
     Randomness Mode:ENiagaraRandomnessMode=::NewEnumerator0
     Random Seed:NiagaraInt32=0
     Fixed Random Seed:NiagaraBool=false
     Bool Selection Array:NiagaraDataInterfaceArrayBool=DataInterface() {propertyValues {GpuSyncMode=SyncCpuToGpu, MaxElements=0, BoolData [false,true]}}
    )) {isSetParameters=true, module=SetVariables_#4}
  ]
  particleUpdateScript [
   &ParticleState()
   &SkeletalMeshLocation(
    Bone / Socket Index=$Particles.PartitionParticles.Partition
    Bone Sampling Mode=::NewEnumerator4
    MaskSpawnGroup=false
    Orientation Sampling=::NewEnumerator1
    Sampled Position Offset=&Multiply_Vector(
     A=&MergeHelper_VectorSwitch(
      0=&SelectVectorFromArray(
       Array Sampling Mode=::NewEnumerator0
       Direct Array Index=?
       Vector Selection Array=DataInterface() {propertyValues {FloatData [
          <20.166 -3.4049 2.7431>
          <15.2029 -4.8726 12.6575>
          <14.0276 5.2138 -4.7046>
          <9.9198 -8.4673 -7.2102>
          <-0.8108 2.4371 -12.142>
          <0.0561 -4.4411 -12.1456>
          <22.0664 0.2571 4.0837>
          <10.5096 3.3678 -14.504>
          <13.0242 6.5768 1.7442>
          <5.1675 -7.9007 10.5401>
          <17.9929 -0.7766 6.2635>
          <-1.081 -9.1323 -2.2206>
          <6.9059 10.271 -6.7415>
          <12.6784 6.0171 5.1392>
          <12.4295 -7.4273 -10.6136>
          <15.3472 -0.9828 15.3862>
          <9.5501 3.4056 14.15>
          <16.5812 -0.6923 11.0465>
          <12.3471 -8.1396 -4.0267>
          <20.6375 3.4021 -3.923>
          <5.9046 -8.9944 -5.722>
          <11.0892 -8.2767 6.4782>
          <15.4666 -5.2546 -10.4095>
          <6.0047 -9.3511 -2.2006>
          <4.6058 -9.4163 2.066>
          <5.1214 2.7309 -12.8633>
          <-2.0834 -8.5163 6.2584>
          <9.2061 8.6364 4.8759>
          <12.6014 3.735 -11.5559>
          <2.8959 -6.0907 12.0799>
          <11.8107 -7.7407 9.9776>
          <-3.0399 -5.9634 9.9914>
          <17.3962 -2.6568 -8.0285>
          <13.9282 -5.5703 15.8733>
          <9.0043 9.0165 -4.0112>
          <-1.2912 -2.0773 12.4829>
          <7.7502 8.0854 -9.3727>
          <9.5091 8.5527 -0.2365>
          <8.2489 -8.4682 8.6588>
          <13.2168 3.0736 12.2413>
          <7.3071 -7.39 12.9848>
          <15.8365 2.8995 -6.5126>
          <-0.7377 -9.1742 2.1846>
          <18.5771 -3.6453 -3.8646>
          <6.2344 2.0767 12.9752>
          <1.7805 -9.0814 4.9042>
          <10.686 6.1795 9.0058>
          <4.5709 6.9375 12.1749>
          <2.7475 11.3943 9.915>
          <0.6742 -7.9166 8.8666>
          <10.4434 -7.6775 14.0722>
          <13.8469 4.0116 8.2618>
          <6.3231 -8.3129 -9.5083>
          <7.9192 6.0405 -11.9587>
          <12.9118 -7.8954 4.3863>
          <6.5224 9.855 8.1373>
          <9.5808 -8.8172 3.9387>
          <6.4483 11.7274 3.2903>
          <16.1265 -2.5758 -12.6882>
          <4.0067 -8.5809 7.7639>
          <6.1144 11.174 -2.3396>
          <2.2175 2.3724 12.6648>
          <12.7369 -6.781 -14.9985>
          <6.097 -9.0504 5.2247>
          <18.0172 4.413 -3.0745>
          <19.584 5.3123 0.8068>
          <15.0692 -6.5278 -6.3871>
          <0.7324 7.0371 11.8603>
          <4.4826 -2.7919 12.694>
          <16.0372 1.3463 -9.5858>
          <17.1701 -4.6711 5.8536>
          <4.862 12.8181 6.3164>
          <10.7621 7.0632 -6.3225>
          <16.3128 1.6421 8.0522>
          <7.9491 7.1361 10.7248>
          <12.3369 5.051 -8.6696>
          <22.2105 -2.1806 -2.9003>
          <4.2806 11.6045 -9.1813>
          <15.5335 4.8113 -0.73>
          <18.5302 0.2153 -5.2703>
          <20.1765 4.3884 3.1151>
          <4.9887 -0.6479 -12.7159>
          <14.9551 1.2496 -13.1556>
          <1.5292 -0.5169 -12.6147>
          <16.4728 3.5337 4.2154>
          <-0.1357 -5.4111 11.7754>
          <15.0453 -2.9137 -16.7883>
          <5.6414 -6.4562 -12.8059>
          <12.2571 7.0078 -2.2064>
          <14.0141 1.2091 -16.2451>
          <9.1012 -7.7432 -13.516>
          <1.6259 6.4518 -12.4028>
          <4.5041 12.4332 -5.1369>
          <2.6035 -5.9371 -12.051>
          <15.5014 -5.6916 8.9413>
          <-2.9555 -6.0326 -10.4162>
          <1.6929 -8.1148 -8.6449>
          <1.7944 -9.1363 -4.5832>
          <-2.272 -8.3492 -6.7592>
          <-2.3023 -2.0133 -12.0007>
         ]}}
       Random Seed=0
      )
      1=&SelectVectorFromArray(
       Array Sampling Mode=::NewEnumerator0
       Direct Array Index=?
       Vector Selection Array=DataInterface() {propertyValues {FloatData [
          <21.2437 2.3982 -2.5169>
          <1.2331 -4.9844 -1.8283>
          <2.6364 -1.306 -4.5243>
          <19.417 3.0787 3.1377>
          <13.3441 3.9736 2.6439>
          <10.5818 -0.5759 -3.406>
          <23.5002 0.0015 4.0935>
          <0.516 1.7335 -4.7945>
          <23.8813 3.7209 -0.2736>
          <18.7551 -2.5869 -2.2875>
          <15.3703 -0.4984 -3.2425>
          <17.371 4.2994 0.5044>
          <13.8018 3.6837 -0.9711>
          <23.0445 -2.5399 -2.3821>
          <12.084 2.3624 4.262>
          <10.8814 -4.0263 -1.4>
          <18.8986 -3.0369 3.1956>
          <11.5586 2.2052 -2.3319>
          <26.3535 -3.5187 1.4901>
          <2.8462 3.5336 -2.8304>
          <8.323 1.6621 -2.8166>
          <5.9168 1.6824 -3.2433>
          <18.288 0.0327 4.3554>
          <7.719 -4.4579 3.6112>
          <13.3026 0.8084 -3.1677>
          <9.7731 -4.9064 1.5456>
          <15.8652 -2.9583 -2.2185>
          <3.9998 0.7166 -4.1469>
          <5.2724 -2.5477 -3.2446>
          <15.0071 -4.4206 -0.235>
          <18.3481 -4.1886 0.2139>
          <18.0405 0.6083 -3.1395>
          <5.7788 -5.0014 1.1284>
          <12.83 -2.6951 -2.7393>
          <15.7734 4.0441 2.2747>
          <5.6657 3.8478 -1.156>
          <8.0911 -1.7142 -3.3173>
          <18.9997 3.4393 -1.4175>
          <1.4166 5.2733 -0.7297>
          <7.7587 -4.2829 -1.1119>
          <11.8619 -0.2572 5.1643>
          <21.1846 -3.8662 -0.5129>
          <23.6719 -3.8623 0.7811>
          <16.1676 -4.1208 2.4975>
          <9.4164 0.075 5.2687>
          <6.6902 3.2394 3.7046>
          <11.6109 4.2996 0.8231>
          <25.3862 2.1913 3.5645>
          <9.7022 2.8237 3.9857>
          <5.9493 4.346 1.5029>
          <3.1166 4.96 1.8073>
          <6.8384 1.1014 5.0819>
          <8.8923 4.1005 2.0773>
          <16.025 2.6786 -2.1424>
          <9.8124 3.0971 -1.4875>
          <8.7384 3.9733 -0.1082>
          <23.9742 -2.4298 3.039>
          <26.0644 -1.3771 3.6514>
          <21.0314 4.0712 0.7466>
          <21.442 -2.8535 2.9573>
          <3.7309 -4.612 -1.4604>
          <14.6777 1.4162 4.4929>
          <4.394 -4.2737 4.1896>
          <0.6933 -2.4369 -4.5406>
          <20.636 -0.8151 -3.1493>
          <16.9743 2.208 3.9426>
          <20.7745 0.0034 4.152>
          <22.5287 2.6703 3.0396>
          <3.7159 2.8686 4.656>
          <26.0738 -3.3005 -1.7623>
          <24.9664 2.5969 -3.2539>
          <12.7224 -4.7423 1.26>
          <16.4317 -1.7046 4.379>
          <10.6875 -3.1803 4.7029>
          <13.5046 -3.6056 3.7651>
          <23.1173 0.529 -3.5755>
          <4.4151 -0.0884 5.5779>
          <14.0025 -1.1903 4.8381>
          <7.228 -2.055 5.4462>
          <25.4829 -0.885 -3.5629>
         ]}}
       Random Seed=0
      )
      2=&SelectVectorFromArray(
       Array Sampling Mode=::NewEnumerator0
       Direct Array Index=?
       Vector Selection Array=DataInterface() {propertyValues {FloatData [
          <22.8596 0.1142 3.4905>
          <4.9005 -5.0777 -1.1743>
          <0.5209 2.225 4.5506>
          <20.9645 1.0933 -3.911>
          <20.2642 -3.8819 0.5243>
          <18.0431 1.3865 2.8637>
          <9.152 0.4667 3.2475>
          <0.6276 -1.2402 5.035>
          <22.4552 -3.318 1.609>
          <11.9019 1.2529 2.9642>
          <17.4213 3.6717 1.1443>
          <25.6801 4.0948 1.7971>
          <15.3688 -3.9855 1.0232>
          <24.1887 -3.7586 -1.2822>
          <9.8047 2.7592 -4.0278>
          <8.5683 4.1982 -0.7765>
          <7.6336 2.8004 2.038>
          <19.9707 -1.3491 2.9349>
          <6.5808 -4.177 1.3478>
          <2.7833 5.1276 -0.3202>
          <14.2206 -2.1255 2.8826>
          <25.3925 1.1288 -4.032>
          <15.5916 -3.0948 -3.8241>
          <6.5324 -4.2138 -4.0935>
          <20.2313 4.1021 -0.1022>
          <11.0802 -4.8342 -1.2399>
          <14.8105 4.1598 -0.1208>
          <23.2687 3.0128 -2.6245>
          <1.1271 -4.1333 3.1095>
          <10.7285 3.235 1.2966>
          <9.8982 -3.3034 2.3132>
          <20.8853 2.3547 2.506>
          <12.3265 3.7025 -3.0313>
          <5.153 4.35 -2.0237>
          <18.2081 -3.0208 -3.3284>
          <19.5793 3.4323 -2.6824>
          <14.9872 0.7238 3.1848>
          <23.3619 3.6126 0.5616>
          <0.8505 4.6719 2.2211>
          <12.1226 4.2985 -0.7684>
          <12.6316 1.2083 -4.696>
          <11.6692 -1.4922 3.2593>
          <3.744 -4.3843 1.8481>
          <4.5672 3.5835 1.9786>
          <6.9152 3.0969 -3.857>
          <18.9212 -0.8647 -4.1779>
          <17.7951 -3.3473 1.6595>
          <3.1776 -1.4709 4.303>
          <15.573 -0.32 -4.7222>
          <14.854 2.5968 -3.9332>
          <17.6436 1.6959 -4.1275>
          <10.4014 -0.358 -5.266>
          <16.5366 4.0145 -2.2859>
          <14.6602 -4.6453 -1.3362>
          <16.9174 -1.3178 3.0362>
          <12.478 -4.0933 1.1209>
          <5.342 -2.3774 3.345>
          <7.8227 -1.8072 3.3053>
          <27.0209 -2.8345 -2.2512>
          <5.692 0.8309 3.6478>
          <1.8364 -5.5956 -0.1369>
          <17.9406 -4.3157 -1.1522>
          <3.1441 -5.0637 -3.4018>
          <2.9236 1.9843 4.0056>
          <24.5109 -2.3121 2.606>
          <21.2421 -3.1699 -2.757>
          <14.1158 2.8015 1.9589>
          <23.0776 -1.0562 -3.9034>
          <4.7083 1.4391 -5.0008>
          <8.9262 -4.5983 0.3785>
          <26.7128 -1.4306 3.1486>
          <10.1558 -3.8666 -4.2357>
          <12.907 -4.2249 -3.3375>
          <4.5895 -2.2437 -5.5012>
          <7.8835 -4.9279 -1.9441>
          <27.3285 -3.4583 1.0851>
          <7.5005 0.4694 -5.2493>
          <12.8834 -1.8146 -4.947>
          <7.9807 -2.1369 -5.3883>
          <26.1371 1.4648 3.8583>
         ]}}
       Random Seed=0
      )
      3=&SelectVectorFromArray(
       Array Sampling Mode=::NewEnumerator0
       Direct Array Index=?
       Vector Selection Array=DataInterface() {propertyValues {FloatData [
          <-0.7134 -2.5592 -2.2688>
          <1.9035 3.8382 -1.9808>
          <5.8898 1.8041 -3.748>
          <1.2049 0.8462 4.01>
          <7.4467 -1.7492 3.0881>
          <8.7381 -3.2894 0.776>
          <4.655 -1.1744 -3.6865>
          <9.2274 3.2904 -1.125>
          <3.2854 3.1133 1.7712>
          <5.397 0.938 3.4882>
          <9.1581 1.5886 2.2517>
          <2.6612 -3.1286 -2.0413>
          <8.873 -0.7019 -3.3754>
          <0.5914 0.7538 -3.9702>
          <6.1762 -3.6758 -1.0554>
          <6.7531 3.1739 1.2092>
          <4.0796 -2.4656 2.7296>
          <1.7874 -3.589 1.2459>
          <-0.2404 -2.577 2.6744>
          <5.1643 3.7213 -1.3073>
         ]}}
       Random Seed=0
      )
      4=&SelectVectorFromArray(
       Array Sampling Mode=::NewEnumerator0
       Direct Array Index=?
       Vector Selection Array=DataInterface() {propertyValues {FloatData [
          <1.6482 -3.4717 -1.4729>
          <1.6502 -2.4028 2.7046>
          <5.7008 -1.9999 3.1951>
          <0.0186 2.1891 -3.7534>
          <2.994 -1.2709 -3.6527>
          <-1.9634 -3.1101 0.9565>
          <-0.917 0.9741 3.7944>
          <2.1816 3.1653 2.8663>
          <9.2479 2.9621 -0.6501>
          <8.9447 0.2488 3.4929>
          <6.1604 3.4956 2.1671>
          <0.4548 4.0756 -0.1597>
          <8.7167 -3.0482 1.4666>
          <3.7438 0.3891 4.0631>
          <-1.2008 -1.8371 -3.0711>
          <5.4271 3.3822 -0.8665>
          <2.2932 3.1124 -1.9963>
          <7.6034 -0.6929 -3.3852>
          <4.8509 1.4991 -3.3289>
          <5.4971 -3.5381 -0.9096>
         ]}}
       Random Seed=0
      )
      5=&SelectVectorFromArray(
       Array Sampling Mode=::NewEnumerator0
       Direct Array Index=?
       Vector Selection Array=DataInterface() {propertyValues {FloatData [
          <40.4493 5.8816 -0.0327>
          <34.3185 -4.5179 4.8504>
          <36.155 4.789 4.6116>
          <20.0868 7.9111 -4.4596>
          <44.2205 -5.7562 1.0783>
          <31.5949 5.9981 4.6492>
          <37.2793 -0.0423 6.4061>
          <31.6285 -4.1674 -3.6821>
          <39.9677 -5.4604 0.7165>
          <32.8999 1.6497 -5.7017>
          <40.4069 -3.5991 -3.0132>
          <24.3978 -6.2564 -1.5867>
          <44.4667 -2.7765 -3.831>
          <39.2444 -4.4332 4.3703>
          <37.2192 4.1804 -3.6351>
          <35.5586 -5.3674 -1.0013>
          <25.5926 -3.2136 -5.5353>
          <34.1927 7.0895 -0.6746>
          <40.5763 3.6719 4.4766>
          <30.1723 -5.9297 2.9995>
          <45.1531 3.5242 3.1389>
          <43.1561 1.8823 -3.6835>
          <29.3736 6.6756 -3.8515>
          <44.5472 -3.2385 5.1313>
          <36.983 -1.5234 -4.6845>
          <28.0804 8.4836 1.061>
          <24.4535 5.3876 -6.3382>
          <42.4004 -0.0455 5.9695>
          <32.3515 -0.0865 6.9152>
          <24.3005 8.6201 -2.0147>
         ]}}
       Random Seed=0
      )
      6=&SelectVectorFromArray(
       Array Sampling Mode=::NewEnumerator0
       Direct Array Index=?
       Vector Selection Array=DataInterface() {propertyValues {FloatData [
          <34.8672 -5.6596 -0.0022>
          <40.8105 4.0587 2.8644>
          <43.1133 -5.5387 -0.2162>
          <16.2548 -1.4811 8.0577>
          <37.7509 1.7508 -6.0069>
          <24.2228 -6.4627 0.9203>
          <37.5885 -3.8701 -5.1987>
          <36.3818 0.1149 5.0101>
          <29.8183 7.5322 -3.2438>
          <26.0428 5.287 6.0001>
          <34.438 7.2203 -0.8283>
          <18.8275 7.4412 5.3537>
          <29.7285 7.4232 2.415>
          <38.4979 5.6654 -2.8489>
          <30.7106 -5.5245 -3.8212>
          <31.011 2.0103 6.0826>
          <22.1022 2.2622 7.7327>
          <30.5171 -3.9822 4.0798>
          <43.5147 -0.1365 4.1533>
          <32.4059 4.309 -5.7042>
          <38.9638 -3.9172 2.7995>
          <24.2956 8.7006 1.6907>
          <24.7609 -2.3533 6.2403>
          <43.9706 4.4055 -2.7724>
          <34.9783 5.05 3.5931>
          <19.4605 -5.1575 4.9347>
          <43.3996 -2.516 -5.6678>
          <43.7075 1.9038 -4.9383>
          <31.6381 -0.6866 -7.0112>
          <15.5294 4.04 8.2075>
         ]}}
       Random Seed=0
      )
      7=&SelectVectorFromArray(
       Array Sampling Mode=::NewEnumerator0
       Direct Array Index=?
       Vector Selection Array=DataInterface() {propertyValues {FloatData [
          <24.7577 3.6654 -2.7924>
          <14.8539 -6.0741 -4.0861>
          <19.2329 3.3667 -2.815>
          <31.9263 5.0897 1.9316>
          <30.2501 3.9985 -3.7984>
          <14.4247 -5.4042 4.3023>
          <30.97 1.3653 4.6409>
          <15.5342 -1.7908 -5.7286>
          <35.8945 4.6082 -5.3821>
          <36.9664 -5.8692 1.8682>
          <23.3073 -3.2859 4.7399>
          <45.8658 -6.9575 -0.0175>
          <43.3285 3.0518 7.6494>
          <39.4754 -1.4259 -7.2344>
          <15.2797 1.9878 -4.1532>
          <10.19 -6.2605 -3.0604>
          <11.1145 -0.7042 -5.152>
          <19.5044 -2.8516 -5.944>
          <15.9136 -7.8609 0.3617>
          <31.6861 -6.4519 0.7187>
          <19.0477 -6.1734 3.7739>
          <40.5984 7.9011 -4.7168>
          <14.8228 3.7858 0.4055>
          <39.313 3.6103 5.9829>
          <33.2791 5.8913 -1.6151>
          <10.6465 2.917 2.3028>
          <41.7519 9.9492 1.0271>
          <11.2621 -7.3291 1.0114>
          <21.4755 0.9463 -5.0798>
          <33.7162 -1.2265 -6.5198>
          <35.6371 0.462 5.6372>
          <25.2821 -1.6832 -6.0243>
          <34.4512 -5.5947 -3.2668>
          <19.431 -7.1667 -2.4682>
          <25.545 -7.0156 -1.5055>
          <40.3393 -6.1778 -1.6704>
          <28.9132 0.6052 -5.8854>
          <43.5259 -4.681 -5.7917>
          <26.8003 -6.2141 2.698>
          <12.2803 0.322 4.2757>
          <22.032 -7.3099 1.4044>
          <6.4943 1.029 -4.1846>
          <6.5204 -3.6204 -4.1703>
          <40.8578 -4.6096 4.0713>
          <27.5692 4.1867 1.367>
          <35.0591 4.3341 4.1968>
          <43.8609 -0.0236 -8.7914>
          <37.2309 7.598 -0.5834>
          <29.191 -5.5037 -4.0001>
          <24.1193 2.1725 3.2019>
          <39.6407 -0.8652 6.1287>
          <43.2644 5.2883 -8.3381>
          <38.7539 7.0586 3.4345>
          <10.7471 3.4819 -2.5542>
          <27.1593 -0.6881 4.8023>
          <16.1175 0.6979 3.9701>
          <42.5003 7.4682 5.7325>
          <23.1356 -5.3686 -4.6299>
          <45.2155 -5.1846 4.5699>
          <19.6053 3.2992 1.5412>
          <20.6179 0.3281 4.1964>
          <44.3021 -1.3976 7.1795>
          <30.1195 -3.8544 4.5413>
          <34.5205 -3.3181 4.8577>
          <10.8245 -3.4515 4.7976>
          <17.9841 -2.6966 5.019>
          <6.802 -2.3103 4.9176>
          <6.8644 4.3578 -0.2123>
          <6.7361 -6.1087 1.9701>
          <6.9146 1.9945 3.5026>
         ]}}
       Random Seed=0
      )
      8=&SelectVectorFromArray(
       Array Sampling Mode=::NewEnumerator0
       Direct Array Index=?
       Vector Selection Array=DataInterface() {propertyValues {FloatData [
          <33.4037 -2.4057 6.272>
          <16.5284 -0.9993 5.6868>
          <27.1045 -4.9113 -3.9573>
          <25.1369 1.1606 5.2414>
          <36.5974 2.1639 6.7667>
          <9.8363 3.9144 -0.7573>
          <10.2026 -6.4502 2.6899>
          <38.2353 -6.2559 0.337>
          <31.6592 2.4995 5.4566>
          <28.4351 4.5939 1.796>
          <27.385 1.0678 -4.2808>
          <44.3999 -0.0194 8.7906>
          <43.751 0.8674 -7.5519>
          <40.0749 8.9558 1.2998>
          <30.1171 -6.5691 -0.9289>
          <11.9175 0.5207 4.8481>
          <34.0518 -6.2229 1.7546>
          <33.0961 -3.594 -4.7084>
          <20.7392 0.2799 5.3233>
          <25.1354 3.6504 -1.6709>
          <16.5213 3.0779 3.0931>
          <40.9269 -0.4491 7.7303>
          <17.5487 -5.4848 4.7687>
          <41.5793 -4.4991 5.4829>
          <28.3535 -1.4803 6.1839>
          <23.8741 -7.1036 -1.605>
          <40.3269 -5.2666 -3.0165>
          <11.2599 3.3322 2.7718>
          <37.7837 -1.4899 -5.6637>
          <6.8664 2.9708 -2.6077>
          <42.1787 -6.2964 1.6103>
          <30.5181 -0.9609 -5.0536>
          <6.598 2.1532 3.6965>
          <12.2438 -3.5376 5.222>
          <18.6677 3.4017 -1.3697>
          <33.9857 6.0283 2.1188>
          <37.7167 -2.3312 6.6174>
          <38.112 6.3454 4.8379>
          <21.9479 4.0408 1.4175>
          <16.1579 -7.1043 -2.686>
          <13.9942 2.9024 -2.1099>
          <30.6627 -5.4966 3.8898>
          <35.4956 -5.4916 -2.6194>
          <31.3947 4.3082 -2.7897>
          <6.5486 -6.0957 1.6258>
          <45.3992 -4.9712 5.9135>
          <45.8802 -6.7156 0.3124>
          <37.3652 -5.2327 4.0513>
          <18.1881 0.4853 -4.0908>
          <6.9214 -5.3383 -3.2006>
          <39.82 4.6349 -5.8923>
          <44.9299 -4.3581 -4.9141>
          <45.5497 -6.1123 -2.4801>
          <21.3085 -4.3011 -4.6419>
          <14.1746 -7.5753 1.5116>
          <16.7706 -3.6503 -4.9504>
          <41.5614 -2.138 -5.9782>
          <6.3979 -2.1849 4.5608>
          <41.6929 4.4399 7.6924>
          <19.3835 -7.6854 0.4484>
          <11.4904 -5.69 -3.8408>
          <41.6185 8.9216 -3.5344>
          <35.0591 1.6817 -5.3353>
          <36.0824 5.9993 -3.0395>
          <10.3117 0.6628 -4.1318>
          <22.9248 0.0395 -4.3299>
          <6.9565 -0.8714 -4.9245>
          <22.3877 -5.8661 4.2119>
          <13.5499 -1.167 -4.8321>
          <26.3859 -5.9977 3.7086>
         ]}}
       Random Seed=0
      )
      Index=$Particles.PartitionParticles.Partition
      Num=::NewEnumerator7
     )
     B=&SelectVectorFromArray(
      Direct Array Index=$Particles.PartitionParticles.Partition
      Vector Selection Array=DataInterface() {propertyValues {FloatData [<1 1 1>, <1 -1 -1>, <-1 -1 -1>, <1 1 1>, <-1 -1 1>, <-1 1 1>, <1 -1 1>, <-1 1 1>, <1 -1 1>]}}
     )
    )
    Skeletal Mesh=$System.Skeletal Mesh
   ) {enabled=false, module=SkeletalMeshLocation002}
   &SampleSkeletalMesh(
    AdditionalYawPitchRoll=false
    Bone Sampling Mode:ENiagaraBoneSamplingMode=::NewEnumerator4
    Skeletal Mesh:NiagaraDataInterfaceSkeletalMesh=$System.Skeletal Mesh
   )
   &ShapeLocation(
    Handle Radius=1
    Large Radius=12
    MaskSpawnGroup=false
    Non Uniform Scale=&MakeCustomVectorFromBool(Bool=$Particles.Mirror, True Vector=<1 -1.5 1>, False Vector=<-1 -1.5 1>)
    Offset=&MakeCustomVectorFromBool(Bool=$Particles.Mirror, True Vector=<-0.9794 -7.9938 5.8326>, False Vector=<0.9794 -7.9938 5.8326>)
    Offset Mode=::NewEnumerator0
    Override Seed=false
    Rotation Angle=90
    Rotation Axis=<0 1 0>
    Rotation Mode=::NewEnumerator1
    Shape Origin=$Particles.SampleSkeletalMesh.SampledPosition
    Shape Primitive=::NewEnumerator3
    Surface Distribution=0.5
    U Distribution=0.8
    UseOwnerScale=true
    V Distribution=0.4
   ) {enabled=false}
   &ShapeLocation(
    MaskSpawnGroup=false
    Offset=&MakeCustomVectorFromBool(
     Bool=$Particles.Mirror
     True Vector=&SelectVectorFromArray(
      Array Sampling Mode=::NewEnumerator0
      Direct Array Index=?
      Vector Selection Array=DataInterface() {propertyValues {FloatData [
         <5.0396 -0.9147 4.6938>
         <10.3826 -9 -0.2261>
         <9.08 -9 -3.4551>
         <6.8852 -9 -6.8791>
         <4.3545 -9 -9.9046>
         <1.6421 -9 -12.1911>
         <10.3983 -8.5 2.8509>
         <10.3983 -7.6495 5.2744>
         <9.3081 -5.561 7.4836>
         <7.6758 -2.8551 7.045>
        ]}}
      Random Seed=0
     )
     False Vector=&SelectVectorFromArray(
      Array Sampling Mode=::NewEnumerator0
      Direct Array Index=?
      Vector Selection Array=DataInterface() {propertyValues {FloatData [
         <-5.0396 -0.9147 4.6938>
         <-10.3826 -9 -0.2261>
         <-9.08 -9 -3.4551>
         <-6.8852 -9 -6.8791>
         <-4.3545 -9 -9.9046>
         <-1.6421 -9 -12.1911>
         <-10.3983 -8.5 2.8509>
         <-10.3983 -7.6495 5.2744>
         <-9.3081 -5.561 7.4836>
         <-7.6758 -2.8551 7.045>
        ]}}
      Random Seed=0
     )
    )
    Offset Mode=::NewEnumerator0
    Override Seed=false
    Shape Origin=$Particles.SampleSkeletalMesh.SampledPosition
    Sphere Radius=1
    UseOwnerScale=true
   ) {module=ShapeLocation001}
   &ScaleSpriteSize(Uniform Curve Scale=&FNTimeOfDayScalar(Night Scale=1.25, Morning Scale=1, Evening Scale=1)) {module=ScaleSpriteSize001}
   &ScaleColor(
    Scale Alpha=&Multiply_Float(
     A=&FloatFromCurve(
      FloatCurve=DataInterface() {propertyValues {Curve {keys [
          {interpMode=RCIM_Cubic, tangentMode=RCTM_User, time=0, value=0}
          {arriveTangent=-2.4058, interpMode=RCIM_Cubic, leaveTangent=-2.4058, tangentMode=RCTM_User, time=0.26, value=0.8421}
          {arriveTangent=-1.1775, interpMode=RCIM_Cubic, leaveTangent=-1.1775, time=0.4613, value=0.0645}
          {arriveTangent=-0.1197, interpMode=RCIM_Cubic, leaveTangent=-0.1197, time=0.5898, value=0.4537}
          <at 1 0 RCIM_Cubic>
         ]}}}
      Scale Curve=&RandomRangeFloat(Minimum=0.1, Maximum=0.5)
     )
     B=&FloatFromCurve(
      FloatCurve=DataInterface() {propertyValues {Curve {keys [{tangentMode=RCTM_User, time=0, value=0}, {tangentMode=RCTM_User, time=1, value=1}]}}}
      Scale Curve=&FloatFromCurve(
       FloatCurve=DataInterface() {propertyValues {Curve {keys [
           {tangentMode=RCTM_User, time=0, value=0}
           {arriveTangentWeight=0.1667, interpMode=RCIM_Cubic, leaveTangentWeight=0.1667, tangentMode=RCTM_User, time=1, value=1}
          ]}}}
       CurveIndex=$Emitter.Age
      )
     )
    )
    Scale RGB=&VectorFromFloat(Value=&Multiply_Float(
      A=&RandomRangeFloat(Minimum=0.7, Maximum=0.9, Evaluation Type=::NewEnumerator1)
      B=&SelectFloatFromArray(Direct Array Index=$Engine.QualityLevel, Float Selection Array=DataInterface() {propertyValues {FloatData [1,1,1,1,1]}})
     ))
    ScaleA=true
    ScaleRGB=true
   ) {module=ScaleColor001}
   &ScaleColor(
    Scale Alpha=$System.TODAlpha
    Scale RGB=&MakeVectorFromLinearColorRGB(Color=$System.TODColor)
    ScaleA=true
    ScaleRGB=true
   )
   /SetVariables(
    Particles.CenterFlareColor:LinearColor=&MultiplyLinearColor(
     Linear Color=$Particles.Color
     Scale Factor=&MakeVector4_FromLinearColor(LinearColor=&SelectLinearColorFromArray(
       Array Sampling Mode=::NewEnumerator1
       Direct Array Index=&MakeIntFromFloat(Float=$User.Variant)
       Color Selection Array=DataInterface() {propertyValues {ColorData [
          <rgba 8.0099 9.8007 10 1>
          <rgba 8.0099 9.8007 10 1>
          <rgba 6.8097 6.5104 10 1>
          <rgba 10 9.3031 3.947 1>
          <rgba 10 10 10 1>
          <rgba 8.0099 9.8007 10 1>
         ]}}
       Random Seed=?
      ))
    )
    Particles.CenterFlareSize:Vector2f=&Multiply_Vector2(A=$Particles.SpriteSize, B=&Vector2DFromFloat(Value=0.5))
    Particles.CoronaFlareColor:LinearColor=&MultiplyLinearColor(
     Linear Color=$Particles.Color
     Scale Factor=&MakeVector4_FromLinearColor(LinearColor=<rgba 0.15 0.15 0.15 1>)
    )
    Particles.CoronaFlareSize:Vector2f=&Multiply_Vector2(A=$Particles.SpriteSize, B=&Vector2DFromFloat(Value=0.75))
    Particles.FlickerFlareColor:LinearColor=&MultiplyLinearColor(
     Linear Color=$Particles.Color
     Scale Factor=&MakeVector4_FromLinearColor(LinearColor=&SelectLinearColorFromArray(
       Array Sampling Mode=::NewEnumerator1
       Direct Array Index=&MakeIntFromFloat(Float=$User.Variant)
       Color Selection Array=DataInterface() {propertyValues {ColorData [
          <rgba 0.9441 0.8079 1 1>
          <rgba 0.9441 0.8079 1 1>
          <rgba 0.0234 0 0.5 1>
          <rgba 0.05 0.0344 0.005 1>
          <rgba 0.5 0.5 0.5 1>
          <rgba 0.404 0.4374 0.5 1>
         ]}}
       Random Seed=?
      ))
    )
   ) {isSetParameters=true, module=SetVariables_#5}
  ]
  rendererClasses [&NiagaraSpriteRendererProperties, &NiagaraMeshRendererProperties]
  renderers [
   {rendererClass=&NiagaraSpriteRendererProperties}
   {rendererClass=&NiagaraSpriteRendererProperties, rendererIndex=1}
   {rendererClass=&NiagaraSpriteRendererProperties, rendererIndex=2}
   {rendererClass=&NiagaraMeshRendererProperties, rendererIndex=3}
  ]
 }
 WingBlings {
  emitterUpdateScript [
   &EmitterState()
   &SpawnRate(Spawn Probability=0.5, SpawnRate=0.9, Use Spawn Probability=true)
  ]
  particleSpawnScript [
   &InitializeParticle(
    Color=<rgba 5 5 5 1>
    Lifetime Max=0.7
    Lifetime Min=0.5
    Lifetime Mode=::NewEnumerator1
    Sprite Size Mode=::NewEnumerator3
    Uniform Sprite Size=&RandomRangeFloat(Minimum=15, Maximum=25)
    UsePositionOffset=false
    Write Material Random=true
   )
   &CameraOffset(Camera Offset Amount=5)
  ]
  particleUpdateScript [
   &ParticleState()
   &SkeletalMeshLocation(
    Bone / Socket Index=$Particles.PartitionParticles.Partition
    Bone Sampling Mode=::NewEnumerator4
    MaskSpawnGroup=false
    Orientation Sampling=::NewEnumerator1
    Sampled Position Offset=&Multiply_Vector(
     A=&MergeHelper_VectorSwitch(
      0=&SelectVectorFromArray(
       Array Sampling Mode=::NewEnumerator0
       Direct Array Index=?
       Vector Selection Array=DataInterface() {propertyValues {FloatData [
          <20.166 -3.4049 2.7431>
          <15.2029 -4.8726 12.6575>
          <14.0276 5.2138 -4.7046>
          <9.9198 -8.4673 -7.2102>
          <-0.8108 2.4371 -12.142>
          <0.0561 -4.4411 -12.1456>
          <22.0664 0.2571 4.0837>
          <10.5096 3.3678 -14.504>
          <13.0242 6.5768 1.7442>
          <5.1675 -7.9007 10.5401>
          <17.9929 -0.7766 6.2635>
          <-1.081 -9.1323 -2.2206>
          <6.9059 10.271 -6.7415>
          <12.6784 6.0171 5.1392>
          <12.4295 -7.4273 -10.6136>
          <15.3472 -0.9828 15.3862>
          <9.5501 3.4056 14.15>
          <16.5812 -0.6923 11.0465>
          <12.3471 -8.1396 -4.0267>
          <20.6375 3.4021 -3.923>
          <5.9046 -8.9944 -5.722>
          <11.0892 -8.2767 6.4782>
          <15.4666 -5.2546 -10.4095>
          <6.0047 -9.3511 -2.2006>
          <4.6058 -9.4163 2.066>
          <5.1214 2.7309 -12.8633>
          <-2.0834 -8.5163 6.2584>
          <9.2061 8.6364 4.8759>
          <12.6014 3.735 -11.5559>
          <2.8959 -6.0907 12.0799>
          <11.8107 -7.7407 9.9776>
          <-3.0399 -5.9634 9.9914>
          <17.3962 -2.6568 -8.0285>
          <13.9282 -5.5703 15.8733>
          <9.0043 9.0165 -4.0112>
          <-1.2912 -2.0773 12.4829>
          <7.7502 8.0854 -9.3727>
          <9.5091 8.5527 -0.2365>
          <8.2489 -8.4682 8.6588>
          <13.2168 3.0736 12.2413>
          <7.3071 -7.39 12.9848>
          <15.8365 2.8995 -6.5126>
          <-0.7377 -9.1742 2.1846>
          <18.5771 -3.6453 -3.8646>
          <6.2344 2.0767 12.9752>
          <1.7805 -9.0814 4.9042>
          <10.686 6.1795 9.0058>
          <4.5709 6.9375 12.1749>
          <2.7475 11.3943 9.915>
          <0.6742 -7.9166 8.8666>
          <10.4434 -7.6775 14.0722>
          <13.8469 4.0116 8.2618>
          <6.3231 -8.3129 -9.5083>
          <7.9192 6.0405 -11.9587>
          <12.9118 -7.8954 4.3863>
          <6.5224 9.855 8.1373>
          <9.5808 -8.8172 3.9387>
          <6.4483 11.7274 3.2903>
          <16.1265 -2.5758 -12.6882>
          <4.0067 -8.5809 7.7639>
          <6.1144 11.174 -2.3396>
          <2.2175 2.3724 12.6648>
          <12.7369 -6.781 -14.9985>
          <6.097 -9.0504 5.2247>
          <18.0172 4.413 -3.0745>
          <19.584 5.3123 0.8068>
          <15.0692 -6.5278 -6.3871>
          <0.7324 7.0371 11.8603>
          <4.4826 -2.7919 12.694>
          <16.0372 1.3463 -9.5858>
          <17.1701 -4.6711 5.8536>
          <4.862 12.8181 6.3164>
          <10.7621 7.0632 -6.3225>
          <16.3128 1.6421 8.0522>
          <7.9491 7.1361 10.7248>
          <12.3369 5.051 -8.6696>
          <22.2105 -2.1806 -2.9003>
          <4.2806 11.6045 -9.1813>
          <15.5335 4.8113 -0.73>
          <18.5302 0.2153 -5.2703>
          <20.1765 4.3884 3.1151>
          <4.9887 -0.6479 -12.7159>
          <14.9551 1.2496 -13.1556>
          <1.5292 -0.5169 -12.6147>
          <16.4728 3.5337 4.2154>
          <-0.1357 -5.4111 11.7754>
          <15.0453 -2.9137 -16.7883>
          <5.6414 -6.4562 -12.8059>
          <12.2571 7.0078 -2.2064>
          <14.0141 1.2091 -16.2451>
          <9.1012 -7.7432 -13.516>
          <1.6259 6.4518 -12.4028>
          <4.5041 12.4332 -5.1369>
          <2.6035 -5.9371 -12.051>
          <15.5014 -5.6916 8.9413>
          <-2.9555 -6.0326 -10.4162>
          <1.6929 -8.1148 -8.6449>
          <1.7944 -9.1363 -4.5832>
          <-2.272 -8.3492 -6.7592>
          <-2.3023 -2.0133 -12.0007>
         ]}}
       Random Seed=0
      )
      1=&SelectVectorFromArray(
       Array Sampling Mode=::NewEnumerator0
       Direct Array Index=?
       Vector Selection Array=DataInterface() {propertyValues {FloatData [
          <21.2437 2.3982 -2.5169>
          <1.2331 -4.9844 -1.8283>
          <2.6364 -1.306 -4.5243>
          <19.417 3.0787 3.1377>
          <13.3441 3.9736 2.6439>
          <10.5818 -0.5759 -3.406>
          <23.5002 0.0015 4.0935>
          <0.516 1.7335 -4.7945>
          <23.8813 3.7209 -0.2736>
          <18.7551 -2.5869 -2.2875>
          <15.3703 -0.4984 -3.2425>
          <17.371 4.2994 0.5044>
          <13.8018 3.6837 -0.9711>
          <23.0445 -2.5399 -2.3821>
          <12.084 2.3624 4.262>
          <10.8814 -4.0263 -1.4>
          <18.8986 -3.0369 3.1956>
          <11.5586 2.2052 -2.3319>
          <26.3535 -3.5187 1.4901>
          <2.8462 3.5336 -2.8304>
          <8.323 1.6621 -2.8166>
          <5.9168 1.6824 -3.2433>
          <18.288 0.0327 4.3554>
          <7.719 -4.4579 3.6112>
          <13.3026 0.8084 -3.1677>
          <9.7731 -4.9064 1.5456>
          <15.8652 -2.9583 -2.2185>
          <3.9998 0.7166 -4.1469>
          <5.2724 -2.5477 -3.2446>
          <15.0071 -4.4206 -0.235>
          <18.3481 -4.1886 0.2139>
          <18.0405 0.6083 -3.1395>
          <5.7788 -5.0014 1.1284>
          <12.83 -2.6951 -2.7393>
          <15.7734 4.0441 2.2747>
          <5.6657 3.8478 -1.156>
          <8.0911 -1.7142 -3.3173>
          <18.9997 3.4393 -1.4175>
          <1.4166 5.2733 -0.7297>
          <7.7587 -4.2829 -1.1119>
          <11.8619 -0.2572 5.1643>
          <21.1846 -3.8662 -0.5129>
          <23.6719 -3.8623 0.7811>
          <16.1676 -4.1208 2.4975>
          <9.4164 0.075 5.2687>
          <6.6902 3.2394 3.7046>
          <11.6109 4.2996 0.8231>
          <25.3862 2.1913 3.5645>
          <9.7022 2.8237 3.9857>
          <5.9493 4.346 1.5029>
          <3.1166 4.96 1.8073>
          <6.8384 1.1014 5.0819>
          <8.8923 4.1005 2.0773>
          <16.025 2.6786 -2.1424>
          <9.8124 3.0971 -1.4875>
          <8.7384 3.9733 -0.1082>
          <23.9742 -2.4298 3.039>
          <26.0644 -1.3771 3.6514>
          <21.0314 4.0712 0.7466>
          <21.442 -2.8535 2.9573>
          <3.7309 -4.612 -1.4604>
          <14.6777 1.4162 4.4929>
          <4.394 -4.2737 4.1896>
          <0.6933 -2.4369 -4.5406>
          <20.636 -0.8151 -3.1493>
          <16.9743 2.208 3.9426>
          <20.7745 0.0034 4.152>
          <22.5287 2.6703 3.0396>
          <3.7159 2.8686 4.656>
          <26.0738 -3.3005 -1.7623>
          <24.9664 2.5969 -3.2539>
          <12.7224 -4.7423 1.26>
          <16.4317 -1.7046 4.379>
          <10.6875 -3.1803 4.7029>
          <13.5046 -3.6056 3.7651>
          <23.1173 0.529 -3.5755>
          <4.4151 -0.0884 5.5779>
          <14.0025 -1.1903 4.8381>
          <7.228 -2.055 5.4462>
          <25.4829 -0.885 -3.5629>
         ]}}
       Random Seed=0
      )
      2=&SelectVectorFromArray(
       Array Sampling Mode=::NewEnumerator0
       Direct Array Index=?
       Vector Selection Array=DataInterface() {propertyValues {FloatData [
          <22.8596 0.1142 3.4905>
          <4.9005 -5.0777 -1.1743>
          <0.5209 2.225 4.5506>
          <20.9645 1.0933 -3.911>
          <20.2642 -3.8819 0.5243>
          <18.0431 1.3865 2.8637>
          <9.152 0.4667 3.2475>
          <0.6276 -1.2402 5.035>
          <22.4552 -3.318 1.609>
          <11.9019 1.2529 2.9642>
          <17.4213 3.6717 1.1443>
          <25.6801 4.0948 1.7971>
          <15.3688 -3.9855 1.0232>
          <24.1887 -3.7586 -1.2822>
          <9.8047 2.7592 -4.0278>
          <8.5683 4.1982 -0.7765>
          <7.6336 2.8004 2.038>
          <19.9707 -1.3491 2.9349>
          <6.5808 -4.177 1.3478>
          <2.7833 5.1276 -0.3202>
          <14.2206 -2.1255 2.8826>
          <25.3925 1.1288 -4.032>
          <15.5916 -3.0948 -3.8241>
          <6.5324 -4.2138 -4.0935>
          <20.2313 4.1021 -0.1022>
          <11.0802 -4.8342 -1.2399>
          <14.8105 4.1598 -0.1208>
          <23.2687 3.0128 -2.6245>
          <1.1271 -4.1333 3.1095>
          <10.7285 3.235 1.2966>
          <9.8982 -3.3034 2.3132>
          <20.8853 2.3547 2.506>
          <12.3265 3.7025 -3.0313>
          <5.153 4.35 -2.0237>
          <18.2081 -3.0208 -3.3284>
          <19.5793 3.4323 -2.6824>
          <14.9872 0.7238 3.1848>
          <23.3619 3.6126 0.5616>
          <0.8505 4.6719 2.2211>
          <12.1226 4.2985 -0.7684>
          <12.6316 1.2083 -4.696>
          <11.6692 -1.4922 3.2593>
          <3.744 -4.3843 1.8481>
          <4.5672 3.5835 1.9786>
          <6.9152 3.0969 -3.857>
          <18.9212 -0.8647 -4.1779>
          <17.7951 -3.3473 1.6595>
          <3.1776 -1.4709 4.303>
          <15.573 -0.32 -4.7222>
          <14.854 2.5968 -3.9332>
          <17.6436 1.6959 -4.1275>
          <10.4014 -0.358 -5.266>
          <16.5366 4.0145 -2.2859>
          <14.6602 -4.6453 -1.3362>
          <16.9174 -1.3178 3.0362>
          <12.478 -4.0933 1.1209>
          <5.342 -2.3774 3.345>
          <7.8227 -1.8072 3.3053>
          <27.0209 -2.8345 -2.2512>
          <5.692 0.8309 3.6478>
          <1.8364 -5.5956 -0.1369>
          <17.9406 -4.3157 -1.1522>
          <3.1441 -5.0637 -3.4018>
          <2.9236 1.9843 4.0056>
          <24.5109 -2.3121 2.606>
          <21.2421 -3.1699 -2.757>
          <14.1158 2.8015 1.9589>
          <23.0776 -1.0562 -3.9034>
          <4.7083 1.4391 -5.0008>
          <8.9262 -4.5983 0.3785>
          <26.7128 -1.4306 3.1486>
          <10.1558 -3.8666 -4.2357>
          <12.907 -4.2249 -3.3375>
          <4.5895 -2.2437 -5.5012>
          <7.8835 -4.9279 -1.9441>
          <27.3285 -3.4583 1.0851>
          <7.5005 0.4694 -5.2493>
          <12.8834 -1.8146 -4.947>
          <7.9807 -2.1369 -5.3883>
          <26.1371 1.4648 3.8583>
         ]}}
       Random Seed=0
      )
      3=&SelectVectorFromArray(
       Array Sampling Mode=::NewEnumerator0
       Direct Array Index=?
       Vector Selection Array=DataInterface() {propertyValues {FloatData [
          <-0.7134 -2.5592 -2.2688>
          <1.9035 3.8382 -1.9808>
          <5.8898 1.8041 -3.748>
          <1.2049 0.8462 4.01>
          <7.4467 -1.7492 3.0881>
          <8.7381 -3.2894 0.776>
          <4.655 -1.1744 -3.6865>
          <9.2274 3.2904 -1.125>
          <3.2854 3.1133 1.7712>
          <5.397 0.938 3.4882>
          <9.1581 1.5886 2.2517>
          <2.6612 -3.1286 -2.0413>
          <8.873 -0.7019 -3.3754>
          <0.5914 0.7538 -3.9702>
          <6.1762 -3.6758 -1.0554>
          <6.7531 3.1739 1.2092>
          <4.0796 -2.4656 2.7296>
          <1.7874 -3.589 1.2459>
          <-0.2404 -2.577 2.6744>
          <5.1643 3.7213 -1.3073>
         ]}}
       Random Seed=0
      )
      4=&SelectVectorFromArray(
       Array Sampling Mode=::NewEnumerator0
       Direct Array Index=?
       Vector Selection Array=DataInterface() {propertyValues {FloatData [
          <1.6482 -3.4717 -1.4729>
          <1.6502 -2.4028 2.7046>
          <5.7008 -1.9999 3.1951>
          <0.0186 2.1891 -3.7534>
          <2.994 -1.2709 -3.6527>
          <-1.9634 -3.1101 0.9565>
          <-0.917 0.9741 3.7944>
          <2.1816 3.1653 2.8663>
          <9.2479 2.9621 -0.6501>
          <8.9447 0.2488 3.4929>
          <6.1604 3.4956 2.1671>
          <0.4548 4.0756 -0.1597>
          <8.7167 -3.0482 1.4666>
          <3.7438 0.3891 4.0631>
          <-1.2008 -1.8371 -3.0711>
          <5.4271 3.3822 -0.8665>
          <2.2932 3.1124 -1.9963>
          <7.6034 -0.6929 -3.3852>
          <4.8509 1.4991 -3.3289>
          <5.4971 -3.5381 -0.9096>
         ]}}
       Random Seed=0
      )
      5=&SelectVectorFromArray(
       Array Sampling Mode=::NewEnumerator0
       Direct Array Index=?
       Vector Selection Array=DataInterface() {propertyValues {FloatData [
          <40.4493 5.8816 -0.0327>
          <34.3185 -4.5179 4.8504>
          <36.155 4.789 4.6116>
          <20.0868 7.9111 -4.4596>
          <44.2205 -5.7562 1.0783>
          <31.5949 5.9981 4.6492>
          <37.2793 -0.0423 6.4061>
          <31.6285 -4.1674 -3.6821>
          <39.9677 -5.4604 0.7165>
          <32.8999 1.6497 -5.7017>
          <40.4069 -3.5991 -3.0132>
          <24.3978 -6.2564 -1.5867>
          <44.4667 -2.7765 -3.831>
          <39.2444 -4.4332 4.3703>
          <37.2192 4.1804 -3.6351>
          <35.5586 -5.3674 -1.0013>
          <25.5926 -3.2136 -5.5353>
          <34.1927 7.0895 -0.6746>
          <40.5763 3.6719 4.4766>
          <30.1723 -5.9297 2.9995>
          <45.1531 3.5242 3.1389>
          <43.1561 1.8823 -3.6835>
          <29.3736 6.6756 -3.8515>
          <44.5472 -3.2385 5.1313>
          <36.983 -1.5234 -4.6845>
          <28.0804 8.4836 1.061>
          <24.4535 5.3876 -6.3382>
          <42.4004 -0.0455 5.9695>
          <32.3515 -0.0865 6.9152>
          <24.3005 8.6201 -2.0147>
         ]}}
       Random Seed=0
      )
      6=&SelectVectorFromArray(
       Array Sampling Mode=::NewEnumerator0
       Direct Array Index=?
       Vector Selection Array=DataInterface() {propertyValues {FloatData [
          <34.8672 -5.6596 -0.0022>
          <40.8105 4.0587 2.8644>
          <43.1133 -5.5387 -0.2162>
          <16.2548 -1.4811 8.0577>
          <37.7509 1.7508 -6.0069>
          <24.2228 -6.4627 0.9203>
          <37.5885 -3.8701 -5.1987>
          <36.3818 0.1149 5.0101>
          <29.8183 7.5322 -3.2438>
          <26.0428 5.287 6.0001>
          <34.438 7.2203 -0.8283>
          <18.8275 7.4412 5.3537>
          <29.7285 7.4232 2.415>
          <38.4979 5.6654 -2.8489>
          <30.7106 -5.5245 -3.8212>
          <31.011 2.0103 6.0826>
          <22.1022 2.2622 7.7327>
          <30.5171 -3.9822 4.0798>
          <43.5147 -0.1365 4.1533>
          <32.4059 4.309 -5.7042>
          <38.9638 -3.9172 2.7995>
          <24.2956 8.7006 1.6907>
          <24.7609 -2.3533 6.2403>
          <43.9706 4.4055 -2.7724>
          <34.9783 5.05 3.5931>
          <19.4605 -5.1575 4.9347>
          <43.3996 -2.516 -5.6678>
          <43.7075 1.9038 -4.9383>
          <31.6381 -0.6866 -7.0112>
          <15.5294 4.04 8.2075>
         ]}}
       Random Seed=0
      )
      7=&SelectVectorFromArray(
       Array Sampling Mode=::NewEnumerator0
       Direct Array Index=?
       Vector Selection Array=DataInterface() {propertyValues {FloatData [
          <24.7577 3.6654 -2.7924>
          <14.8539 -6.0741 -4.0861>
          <19.2329 3.3667 -2.815>
          <31.9263 5.0897 1.9316>
          <30.2501 3.9985 -3.7984>
          <14.4247 -5.4042 4.3023>
          <30.97 1.3653 4.6409>
          <15.5342 -1.7908 -5.7286>
          <35.8945 4.6082 -5.3821>
          <36.9664 -5.8692 1.8682>
          <23.3073 -3.2859 4.7399>
          <45.8658 -6.9575 -0.0175>
          <43.3285 3.0518 7.6494>
          <39.4754 -1.4259 -7.2344>
          <15.2797 1.9878 -4.1532>
          <10.19 -6.2605 -3.0604>
          <11.1145 -0.7042 -5.152>
          <19.5044 -2.8516 -5.944>
          <15.9136 -7.8609 0.3617>
          <31.6861 -6.4519 0.7187>
          <19.0477 -6.1734 3.7739>
          <40.5984 7.9011 -4.7168>
          <14.8228 3.7858 0.4055>
          <39.313 3.6103 5.9829>
          <33.2791 5.8913 -1.6151>
          <10.6465 2.917 2.3028>
          <41.7519 9.9492 1.0271>
          <11.2621 -7.3291 1.0114>
          <21.4755 0.9463 -5.0798>
          <33.7162 -1.2265 -6.5198>
          <35.6371 0.462 5.6372>
          <25.2821 -1.6832 -6.0243>
          <34.4512 -5.5947 -3.2668>
          <19.431 -7.1667 -2.4682>
          <25.545 -7.0156 -1.5055>
          <40.3393 -6.1778 -1.6704>
          <28.9132 0.6052 -5.8854>
          <43.5259 -4.681 -5.7917>
          <26.8003 -6.2141 2.698>
          <12.2803 0.322 4.2757>
          <22.032 -7.3099 1.4044>
          <6.4943 1.029 -4.1846>
          <6.5204 -3.6204 -4.1703>
          <40.8578 -4.6096 4.0713>
          <27.5692 4.1867 1.367>
          <35.0591 4.3341 4.1968>
          <43.8609 -0.0236 -8.7914>
          <37.2309 7.598 -0.5834>
          <29.191 -5.5037 -4.0001>
          <24.1193 2.1725 3.2019>
          <39.6407 -0.8652 6.1287>
          <43.2644 5.2883 -8.3381>
          <38.7539 7.0586 3.4345>
          <10.7471 3.4819 -2.5542>
          <27.1593 -0.6881 4.8023>
          <16.1175 0.6979 3.9701>
          <42.5003 7.4682 5.7325>
          <23.1356 -5.3686 -4.6299>
          <45.2155 -5.1846 4.5699>
          <19.6053 3.2992 1.5412>
          <20.6179 0.3281 4.1964>
          <44.3021 -1.3976 7.1795>
          <30.1195 -3.8544 4.5413>
          <34.5205 -3.3181 4.8577>
          <10.8245 -3.4515 4.7976>
          <17.9841 -2.6966 5.019>
          <6.802 -2.3103 4.9176>
          <6.8644 4.3578 -0.2123>
          <6.7361 -6.1087 1.9701>
          <6.9146 1.9945 3.5026>
         ]}}
       Random Seed=0
      )
      8=&SelectVectorFromArray(
       Array Sampling Mode=::NewEnumerator0
       Direct Array Index=?
       Vector Selection Array=DataInterface() {propertyValues {FloatData [
          <33.4037 -2.4057 6.272>
          <16.5284 -0.9993 5.6868>
          <27.1045 -4.9113 -3.9573>
          <25.1369 1.1606 5.2414>
          <36.5974 2.1639 6.7667>
          <9.8363 3.9144 -0.7573>
          <10.2026 -6.4502 2.6899>
          <38.2353 -6.2559 0.337>
          <31.6592 2.4995 5.4566>
          <28.4351 4.5939 1.796>
          <27.385 1.0678 -4.2808>
          <44.3999 -0.0194 8.7906>
          <43.751 0.8674 -7.5519>
          <40.0749 8.9558 1.2998>
          <30.1171 -6.5691 -0.9289>
          <11.9175 0.5207 4.8481>
          <34.0518 -6.2229 1.7546>
          <33.0961 -3.594 -4.7084>
          <20.7392 0.2799 5.3233>
          <25.1354 3.6504 -1.6709>
          <16.5213 3.0779 3.0931>
          <40.9269 -0.4491 7.7303>
          <17.5487 -5.4848 4.7687>
          <41.5793 -4.4991 5.4829>
          <28.3535 -1.4803 6.1839>
          <23.8741 -7.1036 -1.605>
          <40.3269 -5.2666 -3.0165>
          <11.2599 3.3322 2.7718>
          <37.7837 -1.4899 -5.6637>
          <6.8664 2.9708 -2.6077>
          <42.1787 -6.2964 1.6103>
          <30.5181 -0.9609 -5.0536>
          <6.598 2.1532 3.6965>
          <12.2438 -3.5376 5.222>
          <18.6677 3.4017 -1.3697>
          <33.9857 6.0283 2.1188>
          <37.7167 -2.3312 6.6174>
          <38.112 6.3454 4.8379>
          <21.9479 4.0408 1.4175>
          <16.1579 -7.1043 -2.686>
          <13.9942 2.9024 -2.1099>
          <30.6627 -5.4966 3.8898>
          <35.4956 -5.4916 -2.6194>
          <31.3947 4.3082 -2.7897>
          <6.5486 -6.0957 1.6258>
          <45.3992 -4.9712 5.9135>
          <45.8802 -6.7156 0.3124>
          <37.3652 -5.2327 4.0513>
          <18.1881 0.4853 -4.0908>
          <6.9214 -5.3383 -3.2006>
          <39.82 4.6349 -5.8923>
          <44.9299 -4.3581 -4.9141>
          <45.5497 -6.1123 -2.4801>
          <21.3085 -4.3011 -4.6419>
          <14.1746 -7.5753 1.5116>
          <16.7706 -3.6503 -4.9504>
          <41.5614 -2.138 -5.9782>
          <6.3979 -2.1849 4.5608>
          <41.6929 4.4399 7.6924>
          <19.3835 -7.6854 0.4484>
          <11.4904 -5.69 -3.8408>
          <41.6185 8.9216 -3.5344>
          <35.0591 1.6817 -5.3353>
          <36.0824 5.9993 -3.0395>
          <10.3117 0.6628 -4.1318>
          <22.9248 0.0395 -4.3299>
          <6.9565 -0.8714 -4.9245>
          <22.3877 -5.8661 4.2119>
          <13.5499 -1.167 -4.8321>
          <26.3859 -5.9977 3.7086>
         ]}}
       Random Seed=0
      )
      Index=$Particles.PartitionParticles.Partition
      Num=::NewEnumerator7
     )
     B=&SelectVectorFromArray(
      Direct Array Index=$Particles.PartitionParticles.Partition
      Vector Selection Array=DataInterface() {propertyValues {FloatData [<1 1 1>, <1 -1 -1>, <-1 -1 -1>, <1 1 1>, <-1 -1 1>, <-1 1 1>, <1 -1 1>, <-1 1 1>, <1 -1 1>]}}
     )
    )
    Skeletal Mesh=$System.Skeletal Mesh
   ) {enabled=false, module=SkeletalMeshLocation002}
   &SampleSkeletalMesh(
    AdditionalYawPitchRoll=false
    Bone Sampling Mode:ENiagaraBoneSamplingMode=::NewEnumerator4
    Skeletal Mesh:NiagaraDataInterfaceSkeletalMesh=$System.Skeletal Mesh
   )
   &ShapeLocation(
    Additional Quaternion Rotation=true
    Additional Yaw / Pitch / Roll=&MakeCustomVectorFromBool(
     Bool=$User.IsSkydiving
     True Vector=&Lerp_Vector(
      A=<0 0 95>
      B=<0 0 115>
      Alpha=&MakeCustomFloatFromBool(Bool=$System.FastDive, True Float=1, False Float=0)
     )
     False Vector=<0 0 90>
    )
    Handle Radius=2
    Large Radius=12
    MaskSpawnGroup=false
    Non Uniform Scale=/Niagara/DynamicInputs/Helpers/MaskPosition(
     Mask Source:NiagaraInt32=1
     Mask Parameter:NiagaraInt32=&MakeIntFromFloat(Float=$User.Variant)
     Mask:NiagaraInt32=3
     Masked Position:NiagaraPosition=&SelectVectorFromArray(
      Array Sampling Mode=::NewEnumerator0
      Direct Array Index=?
      Vector Selection Array=DataInterface() {propertyValues {FloatData [<-0.5 -1 1.5>, <0.5 -1 1.5>]}}
      Random Seed=0
     )
     Passthrough Position:NiagaraPosition=&SelectVectorFromArray(
      Array Sampling Mode=::NewEnumerator0
      Direct Array Index=?
      Vector Selection Array=DataInterface() {propertyValues {FloatData [<-1 -1.8 1>, <1 -1.8 1>]}}
      Random Seed=0
     )
    )
    Offset=&MaskVector(
     Mask Source=1
     Mask Parameter=&MakeIntFromFloat(Float=$User.Variant)
     Mask=3
     Masked Vector=<0 -6 7>
     Passthrough Vector=<0 -5 0>
    )
    Offset Mode=::NewEnumerator0
    Override Seed=false
    Rotation Angle Type:ENiagara_AngleInput=::NewEnumerator0
    Rotation Mode=::NewEnumerator3
    Rotation Quaternion=$Particles.SampleSkeletalMesh.SampledRotation
    Shape Origin=$Particles.SampleSkeletalMesh.SampledPosition
    Shape Primitive=::NewEnumerator3
    Surface Distribution=0.5
    U Distribution=0.77
    UseOwnerScale=true
    V Distribution=0.4
   )
   &ScaleSpriteSize(Uniform Curve Scale=&FNTimeOfDayScalar(Night Scale=1.25, Morning Scale=1, Evening Scale=1)) {module=ScaleSpriteSize001}
   &ScaleColor(
    Scale Alpha=&Multiply_Float(
     A=&FloatFromCurve(
      FloatCurve=DataInterface() {propertyValues {Curve {keys [
          {interpMode=RCIM_Cubic, tangentMode=RCTM_User, time=0, value=0}
          {arriveTangent=-2.4058, interpMode=RCIM_Cubic, leaveTangent=-2.4058, tangentMode=RCTM_User, time=0.26, value=0.8421}
          {arriveTangent=-1.1775, interpMode=RCIM_Cubic, leaveTangent=-1.1775, time=0.4613, value=0.0645}
          {arriveTangent=-0.1197, interpMode=RCIM_Cubic, leaveTangent=-0.1197, time=0.5898, value=0.4537}
          <at 1 0 RCIM_Cubic>
         ]}}}
      Scale Curve=&RandomRangeFloat(Minimum=0.1, Maximum=0.5)
     )
     B=&FloatFromCurve(
      FloatCurve=DataInterface() {propertyValues {Curve {keys [{tangentMode=RCTM_User, time=0, value=0}, {tangentMode=RCTM_User, time=1, value=1}]}}}
      Scale Curve=&FloatFromCurve(
       FloatCurve=DataInterface() {propertyValues {Curve {keys [
           {tangentMode=RCTM_User, time=0, value=0}
           {arriveTangentWeight=0.1667, interpMode=RCIM_Cubic, leaveTangentWeight=0.1667, tangentMode=RCTM_User, time=1, value=1}
          ]}}}
       CurveIndex=$Emitter.Age
      )
     )
    )
    Scale RGB=&VectorFromFloat(Value=&Multiply_Float(
      A=&RandomRangeFloat(Minimum=0.7, Maximum=0.9, Evaluation Type=::NewEnumerator1)
      B=&SelectFloatFromArray(Direct Array Index=$Engine.QualityLevel, Float Selection Array=DataInterface() {propertyValues {FloatData [1,1,1,1,1]}})
     ))
    ScaleA=true
    ScaleRGB=true
   ) {module=ScaleColor001}
   &ScaleColor(
    Scale Alpha=$System.TODAlpha
    Scale RGB=&MakeVectorFromLinearColorRGB(Color=$System.TODColor)
    ScaleA=true
    ScaleRGB=true
   )
   /SetVariables(
    Particles.CenterFlareColor:LinearColor=&MultiplyLinearColor(
     Linear Color=$Particles.Color
     Scale Factor=&MakeVector4_FromLinearColor(LinearColor=&SelectLinearColorFromArray(
       Array Sampling Mode=::NewEnumerator1
       Direct Array Index=&MakeIntFromFloat(Float=$User.Variant)
       Color Selection Array=DataInterface() {propertyValues {ColorData [
          <rgba 8.0099 9.8007 10 1>
          <rgba 10 0 2.5293 1>
          <rgba 6.8097 6.5104 10 1>
          <rgba 10 9.3031 3.947 1>
          <rgba 8.0099 9.8007 10 1>
          <rgba 8.0099 9.8007 10 1>
         ]}}
       Random Seed=?
      ))
    )
    Particles.CenterFlareSize:Vector2f=&Multiply_Vector2(A=$Particles.SpriteSize, B=&Vector2DFromFloat(Value=0.5))
    Particles.CoronaFlareColor:LinearColor=&MultiplyLinearColor(
     Linear Color=$Particles.Color
     Scale Factor=&MakeVector4_FromLinearColor(LinearColor=<rgba 0.15 0.15 0.15 1>)
    )
    Particles.CoronaFlareSize:Vector2f=&Multiply_Vector2(A=$Particles.SpriteSize, B=&Vector2DFromFloat(Value=0.75))
    Particles.FlickerFlareColor:LinearColor=&MultiplyLinearColor(
     Linear Color=$Particles.Color
     Scale Factor=&MakeVector4_FromLinearColor(LinearColor=&SelectLinearColorFromArray(
       Array Sampling Mode=::NewEnumerator1
       Direct Array Index=&MakeIntFromFloat(Float=$User.Variant)
       Color Selection Array=DataInterface() {propertyValues {ColorData [
          <rgba 0.9441 0.8079 1 1>
          <rgba 0.1 0 0.039 50>
          <rgba 0.0352 0 0.75 1>
          <rgba 0.1 0.0688 0.01 1>
          <rgba 0.75 0.75 0.75 1>
          <rgba 0.6059 0.6561 0.75 1>
         ]}}
       Random Seed=?
      ))
    )
   ) {isSetParameters=true, module=SetVariables_#5}
  ]
  rendererClasses [&NiagaraSpriteRendererProperties, &NiagaraMeshRendererProperties]
  renderers [
   {rendererClass=&NiagaraSpriteRendererProperties}
   {rendererClass=&NiagaraSpriteRendererProperties, rendererIndex=1}
   {rendererClass=&NiagaraSpriteRendererProperties, rendererIndex=2}
   {rendererClass=&NiagaraMeshRendererProperties, rendererIndex=3}
  ]
 }
 WingSparkles {
  emitterUpdateScript [
   &EmitterState()
   &SpawnRate(SpawnRate=1.25, Use Spawn Probability=true)
  ]
  particleSpawnScript [
   &InitializeParticle(
    Lifetime Max=0.4
    Lifetime Min=0.6
    Lifetime Mode=::NewEnumerator1
    Sprite Size Mode=::NewEnumerator3
    Uniform Sprite Size=&RandomRangeFloat(Minimum=15, Maximum=25)
    UsePositionOffset=false
    Write Material Random=true
   )
   &CameraOffset(Camera Offset Amount=5)
  ]
  particleUpdateScript [
   &ParticleState()
   &SkeletalMeshLocation(
    Bone / Socket Index=$Particles.PartitionParticles.Partition
    Bone Sampling Mode=::NewEnumerator4
    MaskSpawnGroup=false
    Orientation Sampling=::NewEnumerator1
    Sampled Position Offset=&Multiply_Vector(
     A=&MergeHelper_VectorSwitch(
      0=&SelectVectorFromArray(
       Array Sampling Mode=::NewEnumerator0
       Direct Array Index=?
       Vector Selection Array=DataInterface() {propertyValues {FloatData [
          <20.166 -3.4049 2.7431>
          <15.2029 -4.8726 12.6575>
          <14.0276 5.2138 -4.7046>
          <9.9198 -8.4673 -7.2102>
          <-0.8108 2.4371 -12.142>
          <0.0561 -4.4411 -12.1456>
          <22.0664 0.2571 4.0837>
          <10.5096 3.3678 -14.504>
          <13.0242 6.5768 1.7442>
          <5.1675 -7.9007 10.5401>
          <17.9929 -0.7766 6.2635>
          <-1.081 -9.1323 -2.2206>
          <6.9059 10.271 -6.7415>
          <12.6784 6.0171 5.1392>
          <12.4295 -7.4273 -10.6136>
          <15.3472 -0.9828 15.3862>
          <9.5501 3.4056 14.15>
          <16.5812 -0.6923 11.0465>
          <12.3471 -8.1396 -4.0267>
          <20.6375 3.4021 -3.923>
          <5.9046 -8.9944 -5.722>
          <11.0892 -8.2767 6.4782>
          <15.4666 -5.2546 -10.4095>
          <6.0047 -9.3511 -2.2006>
          <4.6058 -9.4163 2.066>
          <5.1214 2.7309 -12.8633>
          <-2.0834 -8.5163 6.2584>
          <9.2061 8.6364 4.8759>
          <12.6014 3.735 -11.5559>
          <2.8959 -6.0907 12.0799>
          <11.8107 -7.7407 9.9776>
          <-3.0399 -5.9634 9.9914>
          <17.3962 -2.6568 -8.0285>
          <13.9282 -5.5703 15.8733>
          <9.0043 9.0165 -4.0112>
          <-1.2912 -2.0773 12.4829>
          <7.7502 8.0854 -9.3727>
          <9.5091 8.5527 -0.2365>
          <8.2489 -8.4682 8.6588>
          <13.2168 3.0736 12.2413>
          <7.3071 -7.39 12.9848>
          <15.8365 2.8995 -6.5126>
          <-0.7377 -9.1742 2.1846>
          <18.5771 -3.6453 -3.8646>
          <6.2344 2.0767 12.9752>
          <1.7805 -9.0814 4.9042>
          <10.686 6.1795 9.0058>
          <4.5709 6.9375 12.1749>
          <2.7475 11.3943 9.915>
          <0.6742 -7.9166 8.8666>
          <10.4434 -7.6775 14.0722>
          <13.8469 4.0116 8.2618>
          <6.3231 -8.3129 -9.5083>
          <7.9192 6.0405 -11.9587>
          <12.9118 -7.8954 4.3863>
          <6.5224 9.855 8.1373>
          <9.5808 -8.8172 3.9387>
          <6.4483 11.7274 3.2903>
          <16.1265 -2.5758 -12.6882>
          <4.0067 -8.5809 7.7639>
          <6.1144 11.174 -2.3396>
          <2.2175 2.3724 12.6648>
          <12.7369 -6.781 -14.9985>
          <6.097 -9.0504 5.2247>
          <18.0172 4.413 -3.0745>
          <19.584 5.3123 0.8068>
          <15.0692 -6.5278 -6.3871>
          <0.7324 7.0371 11.8603>
          <4.4826 -2.7919 12.694>
          <16.0372 1.3463 -9.5858>
          <17.1701 -4.6711 5.8536>
          <4.862 12.8181 6.3164>
          <10.7621 7.0632 -6.3225>
          <16.3128 1.6421 8.0522>
          <7.9491 7.1361 10.7248>
          <12.3369 5.051 -8.6696>
          <22.2105 -2.1806 -2.9003>
          <4.2806 11.6045 -9.1813>
          <15.5335 4.8113 -0.73>
          <18.5302 0.2153 -5.2703>
          <20.1765 4.3884 3.1151>
          <4.9887 -0.6479 -12.7159>
          <14.9551 1.2496 -13.1556>
          <1.5292 -0.5169 -12.6147>
          <16.4728 3.5337 4.2154>
          <-0.1357 -5.4111 11.7754>
          <15.0453 -2.9137 -16.7883>
          <5.6414 -6.4562 -12.8059>
          <12.2571 7.0078 -2.2064>
          <14.0141 1.2091 -16.2451>
          <9.1012 -7.7432 -13.516>
          <1.6259 6.4518 -12.4028>
          <4.5041 12.4332 -5.1369>
          <2.6035 -5.9371 -12.051>
          <15.5014 -5.6916 8.9413>
          <-2.9555 -6.0326 -10.4162>
          <1.6929 -8.1148 -8.6449>
          <1.7944 -9.1363 -4.5832>
          <-2.272 -8.3492 -6.7592>
          <-2.3023 -2.0133 -12.0007>
         ]}}
       Random Seed=0
      )
      1=&SelectVectorFromArray(
       Array Sampling Mode=::NewEnumerator0
       Direct Array Index=?
       Vector Selection Array=DataInterface() {propertyValues {FloatData [
          <21.2437 2.3982 -2.5169>
          <1.2331 -4.9844 -1.8283>
          <2.6364 -1.306 -4.5243>
          <19.417 3.0787 3.1377>
          <13.3441 3.9736 2.6439>
          <10.5818 -0.5759 -3.406>
          <23.5002 0.0015 4.0935>
          <0.516 1.7335 -4.7945>
          <23.8813 3.7209 -0.2736>
          <18.7551 -2.5869 -2.2875>
          <15.3703 -0.4984 -3.2425>
          <17.371 4.2994 0.5044>
          <13.8018 3.6837 -0.9711>
          <23.0445 -2.5399 -2.3821>
          <12.084 2.3624 4.262>
          <10.8814 -4.0263 -1.4>
          <18.8986 -3.0369 3.1956>
          <11.5586 2.2052 -2.3319>
          <26.3535 -3.5187 1.4901>
          <2.8462 3.5336 -2.8304>
          <8.323 1.6621 -2.8166>
          <5.9168 1.6824 -3.2433>
          <18.288 0.0327 4.3554>
          <7.719 -4.4579 3.6112>
          <13.3026 0.8084 -3.1677>
          <9.7731 -4.9064 1.5456>
          <15.8652 -2.9583 -2.2185>
          <3.9998 0.7166 -4.1469>
          <5.2724 -2.5477 -3.2446>
          <15.0071 -4.4206 -0.235>
          <18.3481 -4.1886 0.2139>
          <18.0405 0.6083 -3.1395>
          <5.7788 -5.0014 1.1284>
          <12.83 -2.6951 -2.7393>
          <15.7734 4.0441 2.2747>
          <5.6657 3.8478 -1.156>
          <8.0911 -1.7142 -3.3173>
          <18.9997 3.4393 -1.4175>
          <1.4166 5.2733 -0.7297>
          <7.7587 -4.2829 -1.1119>
          <11.8619 -0.2572 5.1643>
          <21.1846 -3.8662 -0.5129>
          <23.6719 -3.8623 0.7811>
          <16.1676 -4.1208 2.4975>
          <9.4164 0.075 5.2687>
          <6.6902 3.2394 3.7046>
          <11.6109 4.2996 0.8231>
          <25.3862 2.1913 3.5645>
          <9.7022 2.8237 3.9857>
          <5.9493 4.346 1.5029>
          <3.1166 4.96 1.8073>
          <6.8384 1.1014 5.0819>
          <8.8923 4.1005 2.0773>
          <16.025 2.6786 -2.1424>
          <9.8124 3.0971 -1.4875>
          <8.7384 3.9733 -0.1082>
          <23.9742 -2.4298 3.039>
          <26.0644 -1.3771 3.6514>
          <21.0314 4.0712 0.7466>
          <21.442 -2.8535 2.9573>
          <3.7309 -4.612 -1.4604>
          <14.6777 1.4162 4.4929>
          <4.394 -4.2737 4.1896>
          <0.6933 -2.4369 -4.5406>
          <20.636 -0.8151 -3.1493>
          <16.9743 2.208 3.9426>
          <20.7745 0.0034 4.152>
          <22.5287 2.6703 3.0396>
          <3.7159 2.8686 4.656>
          <26.0738 -3.3005 -1.7623>
          <24.9664 2.5969 -3.2539>
          <12.7224 -4.7423 1.26>
          <16.4317 -1.7046 4.379>
          <10.6875 -3.1803 4.7029>
          <13.5046 -3.6056 3.7651>
          <23.1173 0.529 -3.5755>
          <4.4151 -0.0884 5.5779>
          <14.0025 -1.1903 4.8381>
          <7.228 -2.055 5.4462>
          <25.4829 -0.885 -3.5629>
         ]}}
       Random Seed=0
      )
      2=&SelectVectorFromArray(
       Array Sampling Mode=::NewEnumerator0
       Direct Array Index=?
       Vector Selection Array=DataInterface() {propertyValues {FloatData [
          <22.8596 0.1142 3.4905>
          <4.9005 -5.0777 -1.1743>
          <0.5209 2.225 4.5506>
          <20.9645 1.0933 -3.911>
          <20.2642 -3.8819 0.5243>
          <18.0431 1.3865 2.8637>
          <9.152 0.4667 3.2475>
          <0.6276 -1.2402 5.035>
          <22.4552 -3.318 1.609>
          <11.9019 1.2529 2.9642>
          <17.4213 3.6717 1.1443>
          <25.6801 4.0948 1.7971>
          <15.3688 -3.9855 1.0232>
          <24.1887 -3.7586 -1.2822>
          <9.8047 2.7592 -4.0278>
          <8.5683 4.1982 -0.7765>
          <7.6336 2.8004 2.038>
          <19.9707 -1.3491 2.9349>
          <6.5808 -4.177 1.3478>
          <2.7833 5.1276 -0.3202>
          <14.2206 -2.1255 2.8826>
          <25.3925 1.1288 -4.032>
          <15.5916 -3.0948 -3.8241>
          <6.5324 -4.2138 -4.0935>
          <20.2313 4.1021 -0.1022>
          <11.0802 -4.8342 -1.2399>
          <14.8105 4.1598 -0.1208>
          <23.2687 3.0128 -2.6245>
          <1.1271 -4.1333 3.1095>
          <10.7285 3.235 1.2966>
          <9.8982 -3.3034 2.3132>
          <20.8853 2.3547 2.506>
          <12.3265 3.7025 -3.0313>
          <5.153 4.35 -2.0237>
          <18.2081 -3.0208 -3.3284>
          <19.5793 3.4323 -2.6824>
          <14.9872 0.7238 3.1848>
          <23.3619 3.6126 0.5616>
          <0.8505 4.6719 2.2211>
          <12.1226 4.2985 -0.7684>
          <12.6316 1.2083 -4.696>
          <11.6692 -1.4922 3.2593>
          <3.744 -4.3843 1.8481>
          <4.5672 3.5835 1.9786>
          <6.9152 3.0969 -3.857>
          <18.9212 -0.8647 -4.1779>
          <17.7951 -3.3473 1.6595>
          <3.1776 -1.4709 4.303>
          <15.573 -0.32 -4.7222>
          <14.854 2.5968 -3.9332>
          <17.6436 1.6959 -4.1275>
          <10.4014 -0.358 -5.266>
          <16.5366 4.0145 -2.2859>
          <14.6602 -4.6453 -1.3362>
          <16.9174 -1.3178 3.0362>
          <12.478 -4.0933 1.1209>
          <5.342 -2.3774 3.345>
          <7.8227 -1.8072 3.3053>
          <27.0209 -2.8345 -2.2512>
          <5.692 0.8309 3.6478>
          <1.8364 -5.5956 -0.1369>
          <17.9406 -4.3157 -1.1522>
          <3.1441 -5.0637 -3.4018>
          <2.9236 1.9843 4.0056>
          <24.5109 -2.3121 2.606>
          <21.2421 -3.1699 -2.757>
          <14.1158 2.8015 1.9589>
          <23.0776 -1.0562 -3.9034>
          <4.7083 1.4391 -5.0008>
          <8.9262 -4.5983 0.3785>
          <26.7128 -1.4306 3.1486>
          <10.1558 -3.8666 -4.2357>
          <12.907 -4.2249 -3.3375>
          <4.5895 -2.2437 -5.5012>
          <7.8835 -4.9279 -1.9441>
          <27.3285 -3.4583 1.0851>
          <7.5005 0.4694 -5.2493>
          <12.8834 -1.8146 -4.947>
          <7.9807 -2.1369 -5.3883>
          <26.1371 1.4648 3.8583>
         ]}}
       Random Seed=0
      )
      3=&SelectVectorFromArray(
       Array Sampling Mode=::NewEnumerator0
       Direct Array Index=?
       Vector Selection Array=DataInterface() {propertyValues {FloatData [
          <-0.7134 -2.5592 -2.2688>
          <1.9035 3.8382 -1.9808>
          <5.8898 1.8041 -3.748>
          <1.2049 0.8462 4.01>
          <7.4467 -1.7492 3.0881>
          <8.7381 -3.2894 0.776>
          <4.655 -1.1744 -3.6865>
          <9.2274 3.2904 -1.125>
          <3.2854 3.1133 1.7712>
          <5.397 0.938 3.4882>
          <9.1581 1.5886 2.2517>
          <2.6612 -3.1286 -2.0413>
          <8.873 -0.7019 -3.3754>
          <0.5914 0.7538 -3.9702>
          <6.1762 -3.6758 -1.0554>
          <6.7531 3.1739 1.2092>
          <4.0796 -2.4656 2.7296>
          <1.7874 -3.589 1.2459>
          <-0.2404 -2.577 2.6744>
          <5.1643 3.7213 -1.3073>
         ]}}
       Random Seed=0
      )
      4=&SelectVectorFromArray(
       Array Sampling Mode=::NewEnumerator0
       Direct Array Index=?
       Vector Selection Array=DataInterface() {propertyValues {FloatData [
          <1.6482 -3.4717 -1.4729>
          <1.6502 -2.4028 2.7046>
          <5.7008 -1.9999 3.1951>
          <0.0186 2.1891 -3.7534>
          <2.994 -1.2709 -3.6527>
          <-1.9634 -3.1101 0.9565>
          <-0.917 0.9741 3.7944>
          <2.1816 3.1653 2.8663>
          <9.2479 2.9621 -0.6501>
          <8.9447 0.2488 3.4929>
          <6.1604 3.4956 2.1671>
          <0.4548 4.0756 -0.1597>
          <8.7167 -3.0482 1.4666>
          <3.7438 0.3891 4.0631>
          <-1.2008 -1.8371 -3.0711>
          <5.4271 3.3822 -0.8665>
          <2.2932 3.1124 -1.9963>
          <7.6034 -0.6929 -3.3852>
          <4.8509 1.4991 -3.3289>
          <5.4971 -3.5381 -0.9096>
         ]}}
       Random Seed=0
      )
      5=&SelectVectorFromArray(
       Array Sampling Mode=::NewEnumerator0
       Direct Array Index=?
       Vector Selection Array=DataInterface() {propertyValues {FloatData [
          <40.4493 5.8816 -0.0327>
          <34.3185 -4.5179 4.8504>
          <36.155 4.789 4.6116>
          <20.0868 7.9111 -4.4596>
          <44.2205 -5.7562 1.0783>
          <31.5949 5.9981 4.6492>
          <37.2793 -0.0423 6.4061>
          <31.6285 -4.1674 -3.6821>
          <39.9677 -5.4604 0.7165>
          <32.8999 1.6497 -5.7017>
          <40.4069 -3.5991 -3.0132>
          <24.3978 -6.2564 -1.5867>
          <44.4667 -2.7765 -3.831>
          <39.2444 -4.4332 4.3703>
          <37.2192 4.1804 -3.6351>
          <35.5586 -5.3674 -1.0013>
          <25.5926 -3.2136 -5.5353>
          <34.1927 7.0895 -0.6746>
          <40.5763 3.6719 4.4766>
          <30.1723 -5.9297 2.9995>
          <45.1531 3.5242 3.1389>
          <43.1561 1.8823 -3.6835>
          <29.3736 6.6756 -3.8515>
          <44.5472 -3.2385 5.1313>
          <36.983 -1.5234 -4.6845>
          <28.0804 8.4836 1.061>
          <24.4535 5.3876 -6.3382>
          <42.4004 -0.0455 5.9695>
          <32.3515 -0.0865 6.9152>
          <24.3005 8.6201 -2.0147>
         ]}}
       Random Seed=0
      )
      6=&SelectVectorFromArray(
       Array Sampling Mode=::NewEnumerator0
       Direct Array Index=?
       Vector Selection Array=DataInterface() {propertyValues {FloatData [
          <34.8672 -5.6596 -0.0022>
          <40.8105 4.0587 2.8644>
          <43.1133 -5.5387 -0.2162>
          <16.2548 -1.4811 8.0577>
          <37.7509 1.7508 -6.0069>
          <24.2228 -6.4627 0.9203>
          <37.5885 -3.8701 -5.1987>
          <36.3818 0.1149 5.0101>
          <29.8183 7.5322 -3.2438>
          <26.0428 5.287 6.0001>
          <34.438 7.2203 -0.8283>
          <18.8275 7.4412 5.3537>
          <29.7285 7.4232 2.415>
          <38.4979 5.6654 -2.8489>
          <30.7106 -5.5245 -3.8212>
          <31.011 2.0103 6.0826>
          <22.1022 2.2622 7.7327>
          <30.5171 -3.9822 4.0798>
          <43.5147 -0.1365 4.1533>
          <32.4059 4.309 -5.7042>
          <38.9638 -3.9172 2.7995>
          <24.2956 8.7006 1.6907>
          <24.7609 -2.3533 6.2403>
          <43.9706 4.4055 -2.7724>
          <34.9783 5.05 3.5931>
          <19.4605 -5.1575 4.9347>
          <43.3996 -2.516 -5.6678>
          <43.7075 1.9038 -4.9383>
          <31.6381 -0.6866 -7.0112>
          <15.5294 4.04 8.2075>
         ]}}
       Random Seed=0
      )
      7=&SelectVectorFromArray(
       Array Sampling Mode=::NewEnumerator0
       Direct Array Index=?
       Vector Selection Array=DataInterface() {propertyValues {FloatData [
          <24.7577 3.6654 -2.7924>
          <14.8539 -6.0741 -4.0861>
          <19.2329 3.3667 -2.815>
          <31.9263 5.0897 1.9316>
          <30.2501 3.9985 -3.7984>
          <14.4247 -5.4042 4.3023>
          <30.97 1.3653 4.6409>
          <15.5342 -1.7908 -5.7286>
          <35.8945 4.6082 -5.3821>
          <36.9664 -5.8692 1.8682>
          <23.3073 -3.2859 4.7399>
          <45.8658 -6.9575 -0.0175>
          <43.3285 3.0518 7.6494>
          <39.4754 -1.4259 -7.2344>
          <15.2797 1.9878 -4.1532>
          <10.19 -6.2605 -3.0604>
          <11.1145 -0.7042 -5.152>
          <19.5044 -2.8516 -5.944>
          <15.9136 -7.8609 0.3617>
          <31.6861 -6.4519 0.7187>
          <19.0477 -6.1734 3.7739>
          <40.5984 7.9011 -4.7168>
          <14.8228 3.7858 0.4055>
          <39.313 3.6103 5.9829>
          <33.2791 5.8913 -1.6151>
          <10.6465 2.917 2.3028>
          <41.7519 9.9492 1.0271>
          <11.2621 -7.3291 1.0114>
          <21.4755 0.9463 -5.0798>
          <33.7162 -1.2265 -6.5198>
          <35.6371 0.462 5.6372>
          <25.2821 -1.6832 -6.0243>
          <34.4512 -5.5947 -3.2668>
          <19.431 -7.1667 -2.4682>
          <25.545 -7.0156 -1.5055>
          <40.3393 -6.1778 -1.6704>
          <28.9132 0.6052 -5.8854>
          <43.5259 -4.681 -5.7917>
          <26.8003 -6.2141 2.698>
          <12.2803 0.322 4.2757>
          <22.032 -7.3099 1.4044>
          <6.4943 1.029 -4.1846>
          <6.5204 -3.6204 -4.1703>
          <40.8578 -4.6096 4.0713>
          <27.5692 4.1867 1.367>
          <35.0591 4.3341 4.1968>
          <43.8609 -0.0236 -8.7914>
          <37.2309 7.598 -0.5834>
          <29.191 -5.5037 -4.0001>
          <24.1193 2.1725 3.2019>
          <39.6407 -0.8652 6.1287>
          <43.2644 5.2883 -8.3381>
          <38.7539 7.0586 3.4345>
          <10.7471 3.4819 -2.5542>
          <27.1593 -0.6881 4.8023>
          <16.1175 0.6979 3.9701>
          <42.5003 7.4682 5.7325>
          <23.1356 -5.3686 -4.6299>
          <45.2155 -5.1846 4.5699>
          <19.6053 3.2992 1.5412>
          <20.6179 0.3281 4.1964>
          <44.3021 -1.3976 7.1795>
          <30.1195 -3.8544 4.5413>
          <34.5205 -3.3181 4.8577>
          <10.8245 -3.4515 4.7976>
          <17.9841 -2.6966 5.019>
          <6.802 -2.3103 4.9176>
          <6.8644 4.3578 -0.2123>
          <6.7361 -6.1087 1.9701>
          <6.9146 1.9945 3.5026>
         ]}}
       Random Seed=0
      )
      8=&SelectVectorFromArray(
       Array Sampling Mode=::NewEnumerator0
       Direct Array Index=?
       Vector Selection Array=DataInterface() {propertyValues {FloatData [
          <33.4037 -2.4057 6.272>
          <16.5284 -0.9993 5.6868>
          <27.1045 -4.9113 -3.9573>
          <25.1369 1.1606 5.2414>
          <36.5974 2.1639 6.7667>
          <9.8363 3.9144 -0.7573>
          <10.2026 -6.4502 2.6899>
          <38.2353 -6.2559 0.337>
          <31.6592 2.4995 5.4566>
          <28.4351 4.5939 1.796>
          <27.385 1.0678 -4.2808>
          <44.3999 -0.0194 8.7906>
          <43.751 0.8674 -7.5519>
          <40.0749 8.9558 1.2998>
          <30.1171 -6.5691 -0.9289>
          <11.9175 0.5207 4.8481>
          <34.0518 -6.2229 1.7546>
          <33.0961 -3.594 -4.7084>
          <20.7392 0.2799 5.3233>
          <25.1354 3.6504 -1.6709>
          <16.5213 3.0779 3.0931>
          <40.9269 -0.4491 7.7303>
          <17.5487 -5.4848 4.7687>
          <41.5793 -4.4991 5.4829>
          <28.3535 -1.4803 6.1839>
          <23.8741 -7.1036 -1.605>
          <40.3269 -5.2666 -3.0165>
          <11.2599 3.3322 2.7718>
          <37.7837 -1.4899 -5.6637>
          <6.8664 2.9708 -2.6077>
          <42.1787 -6.2964 1.6103>
          <30.5181 -0.9609 -5.0536>
          <6.598 2.1532 3.6965>
          <12.2438 -3.5376 5.222>
          <18.6677 3.4017 -1.3697>
          <33.9857 6.0283 2.1188>
          <37.7167 -2.3312 6.6174>
          <38.112 6.3454 4.8379>
          <21.9479 4.0408 1.4175>
          <16.1579 -7.1043 -2.686>
          <13.9942 2.9024 -2.1099>
          <30.6627 -5.4966 3.8898>
          <35.4956 -5.4916 -2.6194>
          <31.3947 4.3082 -2.7897>
          <6.5486 -6.0957 1.6258>
          <45.3992 -4.9712 5.9135>
          <45.8802 -6.7156 0.3124>
          <37.3652 -5.2327 4.0513>
          <18.1881 0.4853 -4.0908>
          <6.9214 -5.3383 -3.2006>
          <39.82 4.6349 -5.8923>
          <44.9299 -4.3581 -4.9141>
          <45.5497 -6.1123 -2.4801>
          <21.3085 -4.3011 -4.6419>
          <14.1746 -7.5753 1.5116>
          <16.7706 -3.6503 -4.9504>
          <41.5614 -2.138 -5.9782>
          <6.3979 -2.1849 4.5608>
          <41.6929 4.4399 7.6924>
          <19.3835 -7.6854 0.4484>
          <11.4904 -5.69 -3.8408>
          <41.6185 8.9216 -3.5344>
          <35.0591 1.6817 -5.3353>
          <36.0824 5.9993 -3.0395>
          <10.3117 0.6628 -4.1318>
          <22.9248 0.0395 -4.3299>
          <6.9565 -0.8714 -4.9245>
          <22.3877 -5.8661 4.2119>
          <13.5499 -1.167 -4.8321>
          <26.3859 -5.9977 3.7086>
         ]}}
       Random Seed=0
      )
      Index=$Particles.PartitionParticles.Partition
      Num=::NewEnumerator7
     )
     B=&SelectVectorFromArray(
      Direct Array Index=$Particles.PartitionParticles.Partition
      Vector Selection Array=DataInterface() {propertyValues {FloatData [<1 1 1>, <1 -1 -1>, <-1 -1 -1>, <1 1 1>, <-1 -1 1>, <-1 1 1>, <1 -1 1>, <-1 1 1>, <1 -1 1>]}}
     )
    )
    Skeletal Mesh=$System.Skeletal Mesh
   ) {enabled=false, module=SkeletalMeshLocation002}
   &SampleSkeletalMesh(
    AdditionalYawPitchRoll=false
    Bone Sampling Mode:ENiagaraBoneSamplingMode=::NewEnumerator4
    Skeletal Mesh:NiagaraDataInterfaceSkeletalMesh=$System.Skeletal Mesh
   )
   &ShapeLocation(
    Additional Quaternion Rotation=true
    Additional Yaw / Pitch / Roll=&MakeCustomVectorFromBool(
     Bool=$User.IsSkydiving
     True Vector=&Lerp_Vector(
      A=<0 0 95>
      B=<0 0 115>
      Alpha=&MakeCustomFloatFromBool(Bool=$System.FastDive, True Float=1, False Float=0)
     )
     False Vector=<0 0 90>
    )
    Handle Radius=2
    Large Radius=12
    MaskSpawnGroup=false
    Non Uniform Scale=&MaskPosition(
     Mask Source:NiagaraInt32=1
     Mask Parameter:NiagaraInt32=&MakeIntFromFloat(Float=$User.Variant)
     Mask:NiagaraInt32=3
     Masked Position:NiagaraPosition=&SelectVectorFromArray(
      Array Sampling Mode=::NewEnumerator0
      Direct Array Index=?
      Vector Selection Array=DataInterface() {propertyValues {FloatData [<-0.5 -1 1.5>, <0.5 -1 1.5>]}}
      Random Seed=0
     )
     Passthrough Position:NiagaraPosition=&SelectVectorFromArray(
      Array Sampling Mode=::NewEnumerator0
      Direct Array Index=?
      Vector Selection Array=DataInterface() {propertyValues {FloatData [<-1 -1.8 1>, <1 -1.8 1>]}}
      Random Seed=0
     )
    )
    Offset=&MaskVector(
     Mask Source=1
     Mask Parameter=&MakeIntFromFloat(Float=$User.Variant)
     Mask=3
     Masked Vector=<0 -6 7>
     Passthrough Vector=<0 -5 0>
    )
    Offset Mode=::NewEnumerator0
    Override Seed=false
    Rotation Angle Type:ENiagara_AngleInput=::NewEnumerator0
    Rotation Mode=::NewEnumerator3
    Rotation Quaternion=$Particles.SampleSkeletalMesh.SampledRotation
    Shape Origin=$Particles.SampleSkeletalMesh.SampledPosition
    Shape Primitive=::NewEnumerator3
    Surface Distribution=0.5
    U Distribution=0.77
    UseOwnerScale=true
    V Distribution=0.4
   )
   &ScaleSpriteSize(Uniform Curve Scale=&FNTimeOfDayScalar(Night Scale=1.25, Morning Scale=1, Evening Scale=1)) {module=ScaleSpriteSize001}
   &ScaleColor(
    Scale Alpha=&Multiply_Float(
     A=&FloatFromCurve(
      FloatCurve=DataInterface() {propertyValues {Curve {keys [
          {interpMode=RCIM_Cubic, tangentMode=RCTM_User, time=0, value=0}
          {arriveTangent=-2.4058, interpMode=RCIM_Cubic, leaveTangent=-2.4058, tangentMode=RCTM_User, time=0.26, value=0.8421}
          {arriveTangent=-1.1775, interpMode=RCIM_Cubic, leaveTangent=-1.1775, time=0.4613, value=0.0645}
          {arriveTangent=-0.1197, interpMode=RCIM_Cubic, leaveTangent=-0.1197, time=0.5898, value=0.4537}
          <at 1 0 RCIM_Cubic>
         ]}}}
      Scale Curve=&RandomRangeFloat(Minimum=0.1, Maximum=0.5)
     )
     B=&FloatFromCurve(
      FloatCurve=DataInterface() {propertyValues {Curve {keys [{tangentMode=RCTM_User, time=0, value=0}, {tangentMode=RCTM_User, time=1, value=1}]}}}
      Scale Curve=&FloatFromCurve(
       FloatCurve=DataInterface() {propertyValues {Curve {keys [
           {tangentMode=RCTM_User, time=0, value=0}
           {arriveTangentWeight=0.1667, interpMode=RCIM_Cubic, leaveTangentWeight=0.1667, tangentMode=RCTM_User, time=1, value=1}
          ]}}}
       CurveIndex=$Emitter.Age
      )
     )
    )
    Scale RGB=&VectorFromFloat(Value=&Multiply_Float(
      A=&RandomRangeFloat(Minimum=0.7, Maximum=0.9, Evaluation Type=::NewEnumerator1)
      B=&SelectFloatFromArray(Direct Array Index=$Engine.QualityLevel, Float Selection Array=DataInterface() {propertyValues {FloatData [1,1,1,1,1]}})
     ))
    ScaleA=true
    ScaleRGB=true
   ) {module=ScaleColor001}
   &ScaleColor(
    Scale Alpha=$System.TODAlpha
    Scale RGB=&MakeVectorFromLinearColorRGB(Color=$System.TODColor)
    ScaleA=true
    ScaleRGB=true
   )
   /SetVariables(
    Particles.CenterFlareColor:LinearColor=&MultiplyLinearColor(
     Linear Color=$Particles.Color
     Scale Factor=&MakeVector4_FromLinearColor(LinearColor=&SelectLinearColorFromArray(
       Array Sampling Mode=::NewEnumerator1
       Direct Array Index=&MakeIntFromFloat(Float=$User.Variant)
       Color Selection Array=DataInterface() {propertyValues {ColorData [
          <rgba 8.0099 9.8007 10 1>
          <rgba 10 0 2.5293 1>
          <rgba 6.8097 6.5104 10 1>
          <rgba 10 9.3031 3.947 1>
          <rgba 8.0099 9.8007 10 1>
          <rgba 8.0099 9.8007 10 1>
         ]}}
       Random Seed=?
      ))
    )
    Particles.CenterFlareSize:Vector2f=&Multiply_Vector2(A=$Particles.SpriteSize, B=&Vector2DFromFloat(Value=0.5))
    Particles.CoronaFlareColor:LinearColor=&MultiplyLinearColor(
     Linear Color=$Particles.Color
     Scale Factor=&MakeVector4_FromLinearColor(LinearColor=<rgba 0.15 0.15 0.15 1>)
    )
    Particles.CoronaFlareSize:Vector2f=&Multiply_Vector2(A=$Particles.SpriteSize, B=&Vector2DFromFloat(Value=0.75))
    Particles.FlickerFlareColor:LinearColor=&MultiplyLinearColor(
     Linear Color=$Particles.Color
     Scale Factor=&MakeVector4_FromLinearColor(LinearColor=&SelectLinearColorFromArray(
       Array Sampling Mode=::NewEnumerator1
       Direct Array Index=&MakeIntFromFloat(Float=$User.Variant)
       Color Selection Array=DataInterface() {propertyValues {ColorData [
          <rgba 0.9441 0.8079 1 1>
          <rgba 0.3 0 0.0622 50>
          <rgba 0.0234 0 0.5 1>
          <rgba 0.05 0.0344 0.005 1>
          <rgba 0.5 0.5 0.5 1>
          <rgba 0.404 0.4624 0.5 1>
         ]}}
       Random Seed=?
      ))
    )
   ) {isSetParameters=true, module=SetVariables_#5}
  ]
  rendererClasses [&NiagaraSpriteRendererProperties, &NiagaraMeshRendererProperties]
  renderers [
   {rendererClass=&NiagaraSpriteRendererProperties}
   {rendererClass=&NiagaraSpriteRendererProperties, rendererIndex=1}
   {rendererClass=&NiagaraSpriteRendererProperties, rendererIndex=2}
   {rendererClass=&NiagaraMeshRendererProperties, rendererIndex=3}
  ]
 }
}
summary {
 systemName=NS_Backpack_SaxoPop_Idle
 emitters [
  {emitterName=WingBlings, rendererClasses [&NiagaraSpriteRendererProperties, &NiagaraMeshRendererProperties]}
  {emitterName=HeartBlings, rendererClasses [&NiagaraSpriteRendererProperties, &NiagaraMeshRendererProperties]}
  {emitterName=HeartSparkles, rendererClasses [&NiagaraSpriteRendererProperties, &NiagaraMeshRendererProperties]}
  {emitterName=WingSparkles, rendererClasses [&NiagaraSpriteRendererProperties, &NiagaraMeshRendererProperties]}
  {emitterName=Cold_Smoke, rendererClasses [&NiagaraSpriteRendererProperties]}
 ]
 userVariables [
  User.bisFrontEnd {type=NiagaraBool}
  User.bIsFrontEndPreview {type=NiagaraBool}
  User.IsSkydiving {type=NiagaraBool}
  User.Lifetime_Scale {type=NiagaraFloat, defaultValue=0, description=""}
  User.Variant {type=NiagaraFloat, defaultValue=0, description=""}
 ]
}
