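# Physics asset for the SaxoPop backpack mesh, giving its mirrored left and right feathers capsule bodies for secondary motion.
# Six feathers per side attach through small base bodies to a central dyn_heart body, with free swing and twist at that joint.
# Feathers 03 to 06 add a second segment.
# Feather joints have narrow swing1 and twist limits but wide swing2 limits, so each bends mostly like a hinge.
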PhysicsAsset /BRCosmetics/Accessories/FORT_Backpacks/Backpack_F_MED_SaxoPop/Meshes/F_MED_SaxoPop_Pack_Physics
bodies {
 dyn_feather_01_base_l [{shapeType=Capsule, length=0, radius=0.505, shapeName=dyn_feather_01_base_l_capsule}]
 dyn_feather_01_base_r [{shapeType=Capsule, length=0, radius=0.505, shapeName=dyn_feather_01_base_r_capsule}]
 dyn_feather_01_l [{shapeType=Capsule, length=1.6203, radius=1.0779, shapeName=dyn_feather_01_l_capsule, center=<0.5 0 0>, rotation=<rot 90 0 0>}]
 dyn_feather_01_r [{shapeType=Capsule, length=1.6203, radius=1.0779, shapeName=dyn_feather_01_r_capsule, center=<0.5 0 0>, rotation=<rot 90 0 0>}]
 dyn_feather_02_base_l [{shapeType=Capsule, length=0, radius=0.505, shapeName=dyn_feather_02_base_l_capsule}]
 dyn_feather_02_base_r [{shapeType=Capsule, length=0, radius=0.505, shapeName=dyn_feather_02_base_r_capsule}]
 dyn_feather_02_l [{shapeType=Capsule, length=1.9868, radius=1.3337, shapeName=dyn_feather_02_l_capsule, center=<0.7 0 0>, rotation=<rot 90 0 0>}]
 dyn_feather_02_r [{shapeType=Capsule, length=1.9868, radius=1.3337, shapeName=dyn_feather_02_r_capsule, center=<0.5 0 0>, rotation=<rot 90 0 0>}]
 dyn_feather_031_l [{shapeType=Capsule, length=1.5514, radius=0.8047, shapeName=dyn_feather_031_l_capsule, center=<0.5 0 0>, rotation=<rot 90 0 0>}]
 dyn_feather_031_r [{shapeType=Capsule, length=1.5514, radius=0.8047, shapeName=dyn_feather_031_r_capsule, center=<1 0 0>, rotation=<rot 90 0 0>}]
 dyn_feather_03_base_l [{shapeType=Capsule, length=0, radius=0.505, shapeName=dyn_feather_03_base_l_capsule}]
 dyn_feather_03_base_r [{shapeType=Capsule, length=0, radius=0.505, shapeName=dyn_feather_03_base_r_capsule}]
 dyn_feather_03_l [{shapeType=Capsule, length=2.4432, radius=1.1777, shapeName=dyn_feather_03_l_capsule, center=<0.5 0 0>, rotation=<rot 90 0 0>}]
 dyn_feather_03_r [{shapeType=Capsule, length=2.4432, radius=1.1777, shapeName=dyn_feather_03_r_capsule, center=<0.5 0 0>, rotation=<rot 90 0 0>}]
 dyn_feather_041_l [{shapeType=Capsule, length=2.4664, radius=1.1236, shapeName=dyn_feather_041_l_capsule, center=<0.1 0 0>, rotation=<rot 90 0 0>}]
 dyn_feather_041_r [{shapeType=Capsule, length=2.4664, radius=1.1236, shapeName=dyn_feather_041_r_capsule, center=<0.2 0 0>, rotation=<rot 90 0 0>}]
 dyn_feather_04_base_l [{shapeType=Capsule, length=0, radius=0.505, shapeName=dyn_feather_04_base_l_capsule}]
 dyn_feather_04_base_r [{shapeType=Capsule, length=0, radius=0.505, shapeName=dyn_feather_04_base_r_capsule}]
 dyn_feather_04_l [{shapeType=Capsule, length=3.078, radius=1.3409, shapeName=dyn_feather_04_l_capsule, center=<0.5 0.2892 0>, rotation=<rot 90 0 0>}]
 dyn_feather_04_r [{shapeType=Capsule, length=3.078, radius=1.3409, shapeName=dyn_feather_04_r_capsule, center=<0.5 0 0>, rotation=<rot 90 0 0>}]
 dyn_feather_051_l [{shapeType=Capsule, length=2.7621, radius=1.4035, shapeName=dyn_feather_051_l_capsule, center=<2 0 0>, rotation=<rot 90 0 0>}]
 dyn_feather_051_r [{shapeType=Capsule, length=2.7621, radius=1.4035, shapeName=dyn_feather_051_r_capsule, center=<2 0 0>, rotation=<rot 90 0 0>}]
 dyn_feather_05_base_l [{shapeType=Capsule, length=0, radius=0.505, shapeName=dyn_feather_05_base_l_capsule}]
 dyn_feather_05_base_r [{shapeType=Capsule, length=0, radius=0.505, shapeName=dyn_feather_05_base_r_capsule}]
 dyn_feather_05_l [{shapeType=Capsule, length=3.4237, radius=1.6461, shapeName=dyn_feather_05_l_capsule, center=<0.5 0 0>, rotation=<rot 90 0 0>}]
 dyn_feather_05_r [{shapeType=Capsule, length=3.4237, radius=1.6461, shapeName=dyn_feather_05_r_capsule, center=<1 0 0>, rotation=<rot 90 0 0>}]
 dyn_feather_061_l [{shapeType=Capsule, length=8, radius=1.9735, shapeName=dyn_feather_061_l_capsule, center=<6 -0.9777 -0.0273>, rotation=<rot 90 0 0>}]
 dyn_feather_061_r [{shapeType=Capsule, length=8, radius=1.9735, shapeName=dyn_feather_061_r_capsule, center=<4 0.9777 -0.0273>, rotation=<rot 90 0 0>}]
 dyn_feather_06_base_l [{shapeType=Capsule, length=0, radius=0.505, shapeName=dyn_feather_06_base_l_capsule}]
 dyn_feather_06_base_r [{shapeType=Capsule, length=0, radius=0.505, shapeName=dyn_feather_06_base_r_capsule}]
 dyn_feather_06_l [{shapeType=Capsule, length=5.4572, radius=2.5667, shapeName=dyn_feather_06_l_capsule, center=<0.2 0 0>, rotation=<rot 90 0 0>}]
 dyn_feather_06_r [{shapeType=Capsule, length=5.4573, radius=2.5668, shapeName=dyn_feather_06_r_capsule, center=<0.1 0 0>, rotation=<rot 90 0 0>}]
 dyn_heart [{shapeType=Capsule, length=2, radius=1, rotation=<rot -90 0 0>}]
}
constraints [
 {
  bone1Name=dyn_feather_01_base_l
  bone2Name=dyn_heart
  swing1LimitDegrees=45
  swing1Motion=Free
  swing2LimitDegrees=45
  swing2Motion=Free
  twistLimitDegrees=45
  twistMotion=Free
 }
 {bone1Name=dyn_feather_01_l, bone2Name=dyn_feather_01_base_l, swing1LimitDegrees=2, swing2LimitDegrees=40, twistLimitDegrees=4}
 {
  bone1Name=dyn_feather_01_base_r
  bone2Name=dyn_heart
  swing1LimitDegrees=45
  swing1Motion=Free
  swing2LimitDegrees=45
  swing2Motion=Free
  twistLimitDegrees=45
  twistMotion=Free
 }
 {bone1Name=dyn_feather_01_r, bone2Name=dyn_feather_01_base_r, swing1LimitDegrees=2, swing2LimitDegrees=40, twistLimitDegrees=2}
 {
  bone1Name=dyn_feather_02_base_l
  bone2Name=dyn_heart
  swing1LimitDegrees=45
  swing1Motion=Free
  swing2LimitDegrees=45
  swing2Motion=Free
  twistLimitDegrees=45
  twistMotion=Free
 }
 {bone1Name=dyn_feather_02_l, bone2Name=dyn_feather_02_base_l, swing1LimitDegrees=2, swing2LimitDegrees=40, twistLimitDegrees=4}
 {
  bone1Name=dyn_feather_02_base_r
  bone2Name=dyn_heart
  swing1LimitDegrees=45
  swing1Motion=Free
  swing2LimitDegrees=45
  swing2Motion=Free
  twistLimitDegrees=45
  twistMotion=Free
 }
 {bone1Name=dyn_feather_02_r, bone2Name=dyn_feather_02_base_r, swing1LimitDegrees=2, swing2LimitDegrees=40, twistLimitDegrees=2}
 {
  bone1Name=dyn_feather_03_base_l
  bone2Name=dyn_heart
  swing1LimitDegrees=45
  swing1Motion=Free
  swing2LimitDegrees=45
  swing2Motion=Free
  twistLimitDegrees=45
  twistMotion=Free
 }
 {bone1Name=dyn_feather_03_l, bone2Name=dyn_feather_03_base_l, swing1LimitDegrees=2, swing2LimitDegrees=80, twistLimitDegrees=4}
 {
  bone1Name=dyn_feather_03_base_r
  bone2Name=dyn_heart
  swing1LimitDegrees=45
  swing1Motion=Free
  swing2LimitDegrees=45
  swing2Motion=Free
  twistLimitDegrees=45
  twistMotion=Free
 }
 {bone1Name=dyn_feather_03_r, bone2Name=dyn_feather_03_base_r, swing1LimitDegrees=2, swing2LimitDegrees=80, twistLimitDegrees=4}
 {
  bone1Name=dyn_feather_04_base_l
  bone2Name=dyn_heart
  swing1LimitDegrees=45
  swing1Motion=Free
  swing2LimitDegrees=45
  swing2Motion=Free
  twistLimitDegrees=45
  twistMotion=Free
 }
 {bone1Name=dyn_feather_04_l, bone2Name=dyn_feather_04_base_l, swing1LimitDegrees=2, swing2LimitDegrees=60, twistLimitDegrees=5}
 {
  bone1Name=dyn_feather_04_base_r
  bone2Name=dyn_heart
  swing1LimitDegrees=45
  swing1Motion=Free
  swing2LimitDegrees=45
  swing2Motion=Free
  twistLimitDegrees=45
  twistMotion=Free
 }
 {bone1Name=dyn_feather_04_r, bone2Name=dyn_feather_04_base_r, swing1LimitDegrees=2, swing2LimitDegrees=60, twistLimitDegrees=5}
 {
  bone1Name=dyn_feather_05_base_l
  bone2Name=dyn_heart
  swing1LimitDegrees=45
  swing1Motion=Free
  swing2LimitDegrees=45
  swing2Motion=Free
  twistLimitDegrees=45
  twistMotion=Free
 }
 {bone1Name=dyn_feather_05_l, bone2Name=dyn_feather_05_base_l, swing1LimitDegrees=2, swing2LimitDegrees=65, twistLimitDegrees=4}
 {
  bone1Name=dyn_feather_05_base_r
  bone2Name=dyn_heart
  swing1LimitDegrees=45
  swing1Motion=Free
  swing2LimitDegrees=45
  swing2Motion=Free
  twistLimitDegrees=45
  twistMotion=Free
 }
 {bone1Name=dyn_feather_05_r, bone2Name=dyn_feather_05_base_r, swing1LimitDegrees=2, swing2LimitDegrees=65, twistLimitDegrees=4}
 {
  bone1Name=dyn_feather_06_base_l
  bone2Name=dyn_heart
  swing1LimitDegrees=45
  swing1Motion=Free
  swing2LimitDegrees=45
  swing2Motion=Free
  twistLimitDegrees=45
  twistMotion=Free
 }
 {bone1Name=dyn_feather_06_l, bone2Name=dyn_feather_06_base_l, swing1LimitDegrees=3, swing2LimitDegrees=45, twistLimitDegrees=5}
 {
  bone1Name=dyn_feather_06_base_r
  bone2Name=dyn_heart
  swing1LimitDegrees=45
  swing1Motion=Free
  swing2LimitDegrees=45
  swing2Motion=Free
  twistLimitDegrees=45
  twistMotion=Free
 }
 {bone1Name=dyn_feather_06_r, bone2Name=dyn_feather_06_base_r, swing1LimitDegrees=3, swing2LimitDegrees=45, twistLimitDegrees=5}
 {bone1Name=dyn_feather_061_r, bone2Name=dyn_feather_06_r, swing1LimitDegrees=2, swing2LimitDegrees=60, twistLimitDegrees=4}
 {bone1Name=dyn_feather_061_l, bone2Name=dyn_feather_06_l, swing1LimitDegrees=2, swing2LimitDegrees=60, twistLimitDegrees=4}
 {bone1Name=dyn_feather_051_r, bone2Name=dyn_feather_05_r, swing1LimitDegrees=3, swing2LimitDegrees=60, twistLimitDegrees=4}
 {bone1Name=dyn_feather_051_l, bone2Name=dyn_feather_05_l, swing1LimitDegrees=3, swing2LimitDegrees=60, twistLimitDegrees=4}
 {bone1Name=dyn_feather_041_r, bone2Name=dyn_feather_04_r, swing1LimitDegrees=2, swing2LimitDegrees=60, twistLimitDegrees=4}
 {bone1Name=dyn_feather_041_l, bone2Name=dyn_feather_04_l, swing1LimitDegrees=2, swing2LimitDegrees=60, twistLimitDegrees=4}
 {bone1Name=dyn_feather_031_r, bone2Name=dyn_feather_03_r, swing1LimitDegrees=2, swing2LimitDegrees=60, twistLimitDegrees=3}
 {bone1Name=dyn_feather_031_l, bone2Name=dyn_feather_03_l, swing1LimitDegrees=2, swing2LimitDegrees=60, twistLimitDegrees=3}
]
